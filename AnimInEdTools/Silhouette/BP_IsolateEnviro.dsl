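; BP_IsolateEnviro is an editor-tool Blueprint, under AnimInEdTools, that sets the color of a silhouette effect.
; Its SetSilhouetteColor function takes a linear color and writes it to the SilhouetteColor vector parameter of the MPC_Silhouette material parameter collection.
; Materials reading that collection pick up the color globally.
; BeginPlay, Tick, overlap and construction script are empty.
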(event EventBeginPlay)

(event Collision|EventActorBeginOverlap
  (params (OtherActor :type Actor%20Object%20Reference)))

(event EventTick
  (params (DeltaSeconds :type Float%20%28single-precision%29)))

(fn SetSilhouetteColor
  (params (InColor :type Linear%20Color%20Structure))
  (Rendering|Material|SetVectorParameterValue "/AnimInEdTools/Silhouette/MPC_Silhouette.MPC_Silhouette" "SilhouetteColor" InColor :!class /Script/Engine.KismetMaterialLibrary))

(fn ConstructionScript)
