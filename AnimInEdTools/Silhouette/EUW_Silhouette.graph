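# Editor utility widget for animators that silhouettes or isolates actors in the focused Level Sequence, with reset and toggles for lights, shadows, grid and stage.
# Isolate hides level actors, sparing selected actors or control rig owners, and spawns a BP_IsolateEnviro backdrop with adjustable color.
# Silhouette additionally switches bound skeletal meshes from main-pass to depth-pass rendering and sets custom stencil, with color driven by the MPC_Silhouette parameter collection.
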EditorUtilityWidgetBlueprint /AnimInEdTools/Silhouette/EUW_Silhouette
graphs [
 AddSkyDome {nodes [
   AddSkyDome() {is=title K2Node_FunctionEntry}
   > n1 = GetAllActorsOfClass!(execute, WorldContextObject:Object Reference, ActorClass:Actor Class Reference=Get Enviro Actor() -> (Enviro Actor:Actor Class Reference)) -> (then, OutActors:Array of Actor Object References)
   > n2 = Branch(Condition="integer < integer"(A=Length(TargetArray=@n1), B=1))
   > n3 = GetActorArrayAverageLocation!(execute, Actors:Array of Actor Object References=Get IsolatedActors() -> (IsolatedActors:Array of Actor Object References)) -> (then, ReturnValue:Vector)
   > SpawnActorFromClass!(execute, ActorClass:Actor Class Reference=Get Enviro Actor(), Location:Vector=@n3, Rotation:Rotator, bTransient:bool) -> (then, ReturnValue:Actor Object Reference)
   > n4 = GetActorOfClass!(execute, WorldContextObject:Object Reference, ActorClass:Actor Class Reference=&/BP_IsolateEnviro.BP_IsolateEnviro_C) -> (then, ReturnValue:BP Isolate Enviro Object Reference)
   n5 = Get BackgroundColor() -> (BackgroundColor:Color Structure)
   > SetVectorParameterValueOnMaterials!(
    execute
    self:Mesh Component Object Reference=Get SM_Sphere!(self:BP Isolate Enviro Object Reference=@n4) -> (SM_Sphere:Static Mesh Component Object Reference)
    ParameterName:Name=Base Color
    ParameterValue:Vector="To Vector (LinearColor)"!(InLinearColor:Linear Color Structure="To LinearColor (Color)"!(InColor:Color Structure=@n5) -> (ReturnValue:Linear Color Structure)) -> (ReturnValue:Vector)
   ) -> (then)
   > ApplySilhouetteColor()
   > n6 = GetActorOfClass(ActorClass=&BP_IsolateEnviro_C)
   n7 = Get ShowGrid_CHK() -> (ShowGrid_CHK:Editor Utility Check Box Object Reference)
   > SetScalarParameterValueOnMaterials!(
    execute
    self:Mesh Component Object Reference=Get Plane!(self:BP Isolate Enviro Object Reference=@n6) -> (Plane:Static Mesh Component Object Reference)
    ParameterName:Name=GridOpacity
    ParameterValue:float=Select(Option 0:double=Make Literal Float(Value=0), Option 1:double=Make Literal Float(Value=0.2), Index=IsChecked(self=@n7)) -> (ReturnValue:double)
   ) -> (then)
   ApplySilhouetteColor(execute=@n2.else)
  ]}
 CreateSelectedActors_array {nodes [
   CreateSelectedActors_array() {is=title K2Node_FunctionEntry}
   > Clear(TargetArray:Array of Actor Object References=Get selectedActors_array() -> (selectedActors_array:Array of Actor Object References))
   > Sequence!(execute) -> (then_0, then_1, then_2:Exec)
   > n1 = GetSelectedActors()
   > n2 = Branch(Condition="integer >= integer"(A=Length(TargetArray=@n1), B=1))
   > n3 = For Each Loop(Array:Array of Actor Object References=@n1) -> (Array Element:Actor Object Reference)
   > Append Array!(
    execute
    TargetArray:Array of Actor Object References=Get selectedActors_array()
    SourceArray:Array of Actor Object References=Make Array("[0]":Actor Object Reference=@n3) -> (Array:Array of Actor Object References)
   ) -> (then)
   PrintString(execute=@n2.else, InString=no actor selected)
   > n4 = GetFocusedLevelSequence!(execute) -> (then, ReturnValue:Level Sequence Object Reference)
   > n5 = GetControlRigs!(execute, LevelSequence:Level Sequence Object Reference=@n4) -> (then, ReturnValue:Array of Control Rig Sequencer Binding Proxy Structures)
   > Branch(Condition="integer > integer"(A=Length(TargetArray:Array of Control Rig Sequencer Binding Proxy Structures=@n5)))
   > n6 = For Each Loop!(Exec, Array:Array of Control Rig Sequencer Binding Proxy Structures=@n5) -> (
    LoopBody
    Array Element_Proxy:Movie Scene Binding Proxy Structure
    Array Element_ControlRig:Control Rig Object Reference
    Array Element_Track:Movie Scene Control Rig Parameter Track Object Reference
    Array Index
    Completed
   )
   n7 = Get selectedActors_array()
   n8 = Branch(execute=@n6.Completed, Condition="integer > integer"(A=Length(TargetArray:Array of Actor Object References=@n7)))
   PrintString(execute=@n8.else, InString=no control selected)
   > n9 = GetFocusedLevelSequence()
   > n10 = GetBindings!(execute, Sequence:Movie Scene Sequence Object Reference=@n9) -> (then, ReturnValue:Array of Movie Scene Binding Proxy Structures)
   > n11 = For Each Loop!(Exec, Array:Array of Movie Scene Binding Proxy Structures=@n10) -> (LoopBody, Array Element:Movie Scene Binding Proxy Structure, Array Index, Completed)
   > n12 = GetBindingID!(execute, Sequence:Movie Scene Sequence Object Reference, InBinding:"Movie Scene Binding Proxy Structure (by ref)"=@n11) -> (then, ReturnValue:Movie Scene Object Binding ID Structure)
   > n13 = For Each Loop(Array:Array of Object References=GetBoundObjects!(ObjectBinding:Movie Scene Object Binding ID Structure=@n12) -> (ReturnValue:Array of Object References)) -> (Array Element:Object Reference)
   > n14 = Cast To Actor!(execute, Object=@n13) -> (then, CastFailed, AsActor:Actor Object Reference) {is=title K2Node_DynamicCast}
   > Append Array(TargetArray=Get selectedActors_array(), SourceArray=Make Array("[0]"=@n14))
   n15 = CurrentControlSelection!(self:Control Rig Object Reference=@n6.Array Element_ControlRig) -> (ReturnValue:Array of Names)
   Branch(execute=@n6.LoopBody, Condition="integer > integer"(A=Length(TargetArray:Array of Names=@n15)))
   > n16 = GetParent!(execute, InBinding:"Movie Scene Binding Proxy Structure (by ref)"=@n6) -> (then, ReturnValue:Movie Scene Binding Proxy Structure)
   > n17 = GetBindingID(
    Sequence=@n4
    InBinding=Select(
     Option 0:"Movie Scene Binding Proxy Structure (by ref)"=@n6
     Option 1:"Movie Scene Binding Proxy Structure (by ref)"=@n16
     Index=IsValid!(InBinding:"Movie Scene Binding Proxy Structure (by ref)"=@n16) -> (ReturnValue)
    ) -> (ReturnValue:"Movie Scene Binding Proxy Structure (by ref)")
   )
   > n18 = Cast To Actor(Object="Get (a copy)"(Array:Array of Object References=GetBoundObjects(ObjectBinding=@n17)) -> (Output:Object Reference))
   n19 = Get selectedActors_array()
   Append Array(execute=[@n18.then, @n20.else], TargetArray=@n19, SourceArray=Make Array("[0]"=@n18))
   n20 = Branch(Condition=Contains Item(TargetArray:Array of Actor Object References=@n19, ItemToFind:"Actor Object Reference (by ref)"=@n18))
   n21 = For Each Loop(Array:Array of Actor Object References=Get selectedActors_array()) -> (Array Element:Actor Object Reference)
   > PrintString(InString=GetDisplayName(Object=@n21))
  ]}
 CullLightsFromArray {nodes [
   n1 = CullLightsFromArray!() -> (then, array_elem:Actor Object Reference) {is=title K2Node_FunctionEntry}
   > n2 = Branch(Condition=IsValid!(Object:Object Reference=GetComponentByClass!(self:Actor Object Reference=@n1, ComponentClass:Actor Component Class Reference=/Script/Engine.DirectionalLightComponent) -> (ReturnValue:Directional Light Component Object Reference)) -> (ReturnValue))
   n3 = Branch(
    execute=@n2.else
    Condition=IsValid(Object=GetComponentByClass(self=@n1, ComponentClass=/Script/Engine.SkyLightComponent) -> (ReturnValue:Sky Light Component Object Reference))
   )
   n4 = Branch(
    execute=@n3.else
    Condition=IsValid(Object=GetComponentByClass(self=@n1, ComponentClass=/Script/Engine.RectLightComponent) -> (ReturnValue:Rect Light Component Object Reference))
   )
   Return Node!(execute=@n4.else, array_elem1:Actor Object Reference=@n1)
  ]}
 EventGraph {nodes [
   "On Clicked (SilhouetteMode_BTN)"() {is=title K2Node_ComponentBoundEvent}
   > SetBackgroundColor!(
    execute
    self:Button Object Reference=Get SilhouetteMode_BTN() -> (SilhouetteMode_BTN:Editor Utility Button Object Reference)
    InBackgroundColor:Linear Color Structure=(R=0.1146,G=0.1146,B=0.1146,A=1)
   ) -> (then)
   > SetBackgroundColor(self=Get Isolate_BTN() -> (Isolate_BTN:Editor Utility Button Object Reference), InBackgroundColor=(R=1,G=1,B=1,A=1))
   > ApplySilhouetteColor()
   > ExecuteConsoleCommand(Command=ShowFlag.BillboardSprites 0)
   > n1 = Sequence!(execute) -> (then_0, then_1, then_2, then_3:Exec, then_4:Exec)
   Set ToolEnabled(execute=[@n1.then_0, @n8.then_0, @n3.then_0], ToolEnabled:bool=false) -> (Output_Get:bool)
   > n2 = Sequence()
   > RestoreSequencer()
   SilhouetteSequencer(execute=@n2.then_1)
   RemoveIsoJunk(execute=@n2.then_3)
   Set ToolEnabled(execute=@n1.then_1, ToolEnabled=true)
   > Clear(TargetArray=Get hide_array() -> (hide_array:Array of Actor Object References))
   > IsolateSequencer()
   SilhouetteSequencer(execute=@n1.then_2)
   AddSkyDome(execute=@n1.then_3) {is=title K2Node_CallFunction}
   Lights_Shadow!(
    execute=@n1.then_4
    self
    Lights:bool=IsChecked(self=Get HideLight_CHK() -> (HideLight_CHK:Editor Utility Check Box Object Reference))
    Shadow:bool=IsChecked(self=Get HideShadow_CHK() -> (HideShadow_CHK:Editor Utility Check Box Object Reference))
   ) -> (then)
   "On Clicked (Isolate_BTN)"() {is=title K2Node_ComponentBoundEvent}
   > SetBackgroundColor(self=Get Isolate_BTN(), InBackgroundColor=(R=0.1146,G=0.1146,B=0.1146,A=1))
   > SetBackgroundColor(self=Get SilhouetteMode_BTN(), InBackgroundColor=(R=1,G=1,B=1,A=1))
   > ExecuteConsoleCommand(Command=ShowFlag.BillboardSprites 0)
   > n3 = Sequence()
   Set ToolEnabled(execute=@n3.then_1, ToolEnabled=true)
   > GetSequencerContents()
   IsolateSequencer(execute=@n3.then_2)
   AddSkyDome(execute=@n3.then_3)
   Lights_Shadow(execute=@n3.then_4, Lights=IsChecked(self=Get HideLight_CHK()), Shadow=IsChecked(self=Get HideShadow_CHK()))
   n4 = "On Check State Changed (ShowStage_CHK)"!() -> (OutputDelegate, then, bIsChecked:bool) {is=title K2Node_ComponentBoundEvent}
   > n5 = Set hideStage(hideStage:bool=@n4.bIsChecked) -> (Output_Get:bool)
   > n6 = GetAllActorsOfClass(ActorClass=/AnimInEdTools/AnimationBrowser/Assets/AnimationBrowserStage.AnimationBrowserStage_C) -> (OutActors:Array of Animation Browser Stage Object References)
   > n7 = For Each Loop(Array:Array of Animation Browser Stage Object References=@n6) -> (Array Element:Animation Browser Stage Object Reference)
   > SetVisibility!(
    execute
    self:Scene Component Object Reference=Get Cube!(self:Animation Browser Stage Object Reference=@n7) -> (Cube:Static Mesh Component Object Reference)
    bNewVisibility:bool=@n5
    bPropagateToChildren:bool=false
   ) -> (then)
   "On Clicked (Reset_BTN)"() {is=title K2Node_ComponentBoundEvent}
   > SetBackgroundColor(self=Get SilhouetteMode_BTN(), InBackgroundColor=(R=1,G=1,B=1,A=1))
   > SetBackgroundColor(self=Get Isolate_BTN(), InBackgroundColor=(R=1,G=1,B=1,A=1))
   > ExecuteConsoleCommand(Command=ShowFlag.BillboardSprites 0)
   > n8 = Sequence!(execute) -> (then_0, then_1)
   Lights_Shadow(execute=@n8.then_1, Lights=IsChecked(self=Get HideLight_CHK()), Shadow=IsChecked(self=Get HideShadow_CHK()))
   "On Property Changed (ColorBG)"!() -> (OutputDelegate, then, PropertyName:Name=None) {is=title K2Node_ComponentBoundEvent}
   > Branch(Condition=Get ToolEnabled() -> (ToolEnabled:bool))
   > n9 = GetActorOfClass(ActorClass=&BP_IsolateEnviro_C)
   n10 = Get BackgroundColor()
   > SetVectorParameterValueOnMaterials(
    self=Get SM_Sphere(self=@n9)
    ParameterName=Base Color
    ParameterValue="To Vector (LinearColor)"(InLinearColor="To LinearColor (Color)"(InColor=@n10))
   )
   n11 = "On Check State Changed (HideLight_CHK)"!() -> (OutputDelegate, then, bIsChecked:bool) {is=title K2Node_ComponentBoundEvent}
   > n12 = SetIsEnabled(self=Get HorizontalBox_Shadows() -> (HorizontalBox_Shadows:Horizontal Box Object Reference), bInIsEnabled=@n11.bIsChecked)
   Lights_Shadow(execute=[@n12.then, @n15.then], Lights=IsChecked(self=Get HideLight_CHK()), Shadow=IsChecked(self=Get HideShadow_CHK()))
   n13 = "On Check State Changed (ShowGrid_CHK)"!() -> (OutputDelegate, then, bIsChecked:bool) {is=title K2Node_ComponentBoundEvent}
   > Branch(Condition=Get ToolEnabled())
   > n14 = GetActorOfClass(ActorClass=&BP_IsolateEnviro_C)
   > SetScalarParameterValueOnMaterials(
    self=Get Plane(self=@n14)
    ParameterName=GridOpacity
    ParameterValue=Select(Option 0:double=Make Literal Float(Value=0), Option 1:double=Make Literal Float(Value=0.2), Index=@n13.bIsChecked) -> (ReturnValue:double)
   )
   n15 = "On Check State Changed (HideShadow_CHK)"!() -> (OutputDelegate, then, bIsChecked:bool) {is=title K2Node_ComponentBoundEvent}
   ApplySilhouetteColor() {is=title K2Node_CustomEvent}
   > SetVectorParameterValue!(
    execute
    WorldContextObject
    Collection:Material Parameter Collection Object Reference=&/MPC_Silhouette
    ParameterName:Name=SilhouetteColor
    ParameterValue:"Linear Color Structure (by ref)"="To LinearColor (Color)"(InColor=Get SilhouetteColor() -> (SilhouetteColor:Color Structure))
   ) -> (then) {is=title K2Node_CallMaterialParameterCollectionFunction}
   Event PreConstruct()
   > Sequence!(execute) -> (then_0, then_1, then_2:Exec)
   > SetObject(self=Get ColorBG() -> (ColorBG:Single Property View Object Reference), NewObject=Self-Reference())
   > SetPropertyName!(execute, self:Single Property View Object Reference=Get ColorBG(), NewPropertyName:Name=BackgroundColor) -> (then)
   Event Construct()
   > SetVectorParameterValue(
    Collection=&MPC_Silhouette
    ParameterName=SilhouetteColor
    ParameterValue="To LinearColor (Color)"(InColor=Get SilhouetteColor())
   )
  ]}
 GetSequencerContents {nodes [
   n1 = For Each Loop(Exec=@n8.then, Array:Array of Actor Object References=Get LSActors_array() -> (LSActors_array:Array of Actor Object References)) -> (Array Element:Actor Object Reference)
   n2 = Get RootComponent!(self:Actor Object Reference=@n1) -> (RootComponent:Scene Component Object Reference)
   n3 = SetVisibility(self=@n2, bNewVisibility=false, bPropagateToChildren=true)
   Add(execute=[@n3.then, @n13.then], TargetArray:Array of Actor Object References=Get hide_array(), NewItem:"Actor Object Reference (by ref)"=@n1) {is=title K2Node_CallArrayFunction}
   GetSequencerContents() {is=title K2Node_FunctionEntry}
   > Clear(TargetArray=Get LSActors_array())
   > Clear(TargetArray=Get hide_array())
   > Clear(TargetArray=Get selectedActors_array())
   > n4 = GetFocusedLevelSequence()
   > n5 = GetBindings(Sequence=@n4)
   > n6 = For Each Loop(Array:Array of Movie Scene Binding Proxy Structures=@n5) -> (Array Element:Movie Scene Binding Proxy Structure)
   CreateSelectedActors_array(execute=@n6.Completed) {is=title K2Node_CallFunction}
   > n7 = GetActorOfClass(ActorClass=&AnimationBrowserStage_C) -> (ReturnValue:Animation Browser Stage Object Reference)
   > n8 = Add(TargetArray=Get selectedActors_array(), NewItem=@n7)
   n9 = GetBindingID(execute=@n6.LoopBody, InBinding=@n6)
   > n10 = Cast To Actor(Object="Get (a copy)"(Array=GetBoundObjects(ObjectBinding=@n9)))
   > Add(TargetArray=Get LSActors_array(), NewItem=@n10)
   n11 = Contains Item(TargetArray=Get selectedActors_array(), ItemToFind=@n1)
   n12 = GetDisplayName(Object=@n2)
   n13 = PrintString(execute=@n14.else, InString=@n12)
   n14 = Branch(execute=@n1.LoopBody, Condition=@n11)
  ]}
 Get_Thumb_Brush {nodes [
   n1 = Get_Thumb_Brush!() -> (then, NewImage:Slate Brush Structure) {is=title K2Node_FunctionEntry}
   > Return Node!(execute, ReturnValue:Slate Brush Structure=@n1)
  ]}
 IsolateSequencer {nodes [
   IsolateSequencer() {is=title K2Node_FunctionEntry}
   > n1 = GetAllLevelActors(self=Get EditorActorSubsystem())
   > n2 = For Each Loop(Array:Array of Actor Object References=@n1) -> (Array Element:Actor Object Reference)
   n3 = For Each Loop(Exec=@n2.Completed, Array=Get hide_array())
   n4 = GetAllActorsOfClass(execute=@n3.Completed, ActorClass=&AnimationBrowserStage_C)
   > n5 = For Each Loop(Array:Array of Animation Browser Stage Object References=@n4) -> (Array Element:Animation Browser Stage Object Reference)
   > SetVisibility(
    self=Get Cube(self=@n5)
    bNewVisibility=IsChecked(self=Get ShowStage_CHK() -> (ShowStage_CHK:Editor Utility Check Box Object Reference))
    bPropagateToChildren=false
   )
   n6 = Get Class Display Name(Class=Get Class(Object=@n3))
   n7 = Branch(execute=@n3.LoopBody, Condition=NOT Boolean(A=Contains(SearchIn=@n6, Substring=Mannequin)))
   > SetVisibility(self=Get RootComponent(self=@n3), bNewVisibility=NOT Boolean(A=Get ToolEnabled()), bPropagateToChildren=true)
   n8 = For Each Loop(
    Exec=@n7.else
    Array:Array of Skeletal Mesh Component Object References=Get Components By Class!(self:Actor Object Reference=@n3, ComponentClass:Actor Component Class Reference=/Script/Engine.SkeletalMeshComponent) -> (ReturnValue:Array of Skeletal Mesh Component Object References)
   ) -> (Array Element:Skeletal Mesh Component Object Reference)
   > SetVisibility(self=@n8, bNewVisibility=NOT Boolean(A=Get ToolEnabled()), bPropagateToChildren=false)
   n9 = CullLightsFromArray!(execute=@n2.LoopBody, self, array_elem:Actor Object Reference=@n2) -> (then, array_elem1:Actor Object Reference) {is=title K2Node_CallFunction}
   > Append Array(TargetArray=Get hide_array(), SourceArray=Make Array("[0]"=@n9))
   NOT Boolean(A=Get ToolEnabled())
  ]}
 Lights_Shadow {nodes [
   n1 = Lights_Shadow!() -> (then, Lights:bool=~, Shadow:bool=~) {is=title K2Node_FunctionEntry}
   > Set loc_light(loc_light:bool=@n1) -> (Output_Get:bool)
   > Set loc_shad(loc_shad:bool=@n1.Shadow) -> (Output_Get:bool)
   > n2 = Sequence!(execute) -> (then_0, then_1, then_2, then_3:Exec, then_4:Exec, then_5:Exec)
   > n3 = GetAllActorsOfClass(ActorClass=/Script/Engine.DirectionalLight) -> (OutActors:Array of Directional Light Object References)
   > n4 = For Each Loop(Array:Array of Actor Object References=@n3) -> (Array Element:Actor Object Reference)
   > SetVisibility(self=Get RootComponent(self=@n4), bNewVisibility=Get loc_light() -> (loc_light:bool), bPropagateToChildren=true)
   > n5 = Cast To DirectionalLight!(execute, Object=@n4) -> (then, CastFailed, AsDirectional Light:Directional Light Object Reference) {is=title K2Node_DynamicCast}
   > SetCastShadows!(
    execute
    self:Light Component Base Object Reference=Get LightComponent!(self:Light Object Reference=@n5) -> (LightComponent:Light Component Object Reference)
    bNewValue:bool=Get loc_shad() -> (loc_shad:bool)
   ) -> (then)
   n6 = GetAllActorsOfClass(execute=@n2.then_1, ActorClass=/Script/Engine.RectLight) -> (OutActors:Array of Rect Light Object References)
   > n7 = For Each Loop(Array=@n6)
   > SetVisibility(self=Get RootComponent(self=@n7), bNewVisibility=Get loc_light(), bPropagateToChildren=true)
   > n8 = Cast To RectLight!(execute, Object=@n7) -> (then, CastFailed, AsRect Light:Rect Light Object Reference) {is=title K2Node_DynamicCast}
   > SetCastShadows(self=Get LightComponent(self=@n8), bNewValue=Get loc_shad())
   n9 = GetAllActorsOfClass(execute=@n2.then_2, ActorClass=/Script/Engine.SpotLight) -> (OutActors:Array of Spot Light Object References)
   > n10 = For Each Loop(Array=@n9)
   > SetVisibility(self=Get RootComponent(self=@n10), bNewVisibility=Get loc_light(), bPropagateToChildren=true)
   > n11 = Cast To SpotLight!(execute, Object=@n10) -> (then, CastFailed, AsSpot Light:Spot Light Object Reference) {is=title K2Node_DynamicCast}
   > SetCastShadows(self=Get LightComponent(self=@n11), bNewValue=Get loc_shad())
   n12 = GetAllActorsOfClass(execute=@n2.then_3, ActorClass=/Script/Engine.PointLight) -> (OutActors:Array of Point Light Object References)
   > n13 = For Each Loop(Array=@n12)
   > SetVisibility(self=Get RootComponent(self=@n13), bNewVisibility=Get loc_light(), bPropagateToChildren=true)
   > n14 = Cast To PointLight!(execute, Object=@n13) -> (then, CastFailed, AsPoint Light:Point Light Object Reference) {is=title K2Node_DynamicCast}
   > SetCastShadows(self=Get LightComponent(self=@n14), bNewValue=Get loc_shad())
   n15 = GetActorOfClass(execute=@n2.then_4, ActorClass=Get Enviro Actor()) -> (ReturnValue:Actor Object Reference)
   n16 = Get Components By Class(self=@n15, ComponentClass=&DirectionalLightComponent) -> (ReturnValue:Array of Directional Light Component Object References)
   > n17 = Branch(Condition="integer > integer"(A=Length(TargetArray:Array of Directional Light Component Object References=@n16)))
   > n18 = Branch(Condition=Get ToolEnabled())
   n19 = For Each Loop(Exec=[@n20.then, @n24.then, @n18.then, @n22.then], Array:Array of Directional Light Component Object References=@n16) -> (Array Element:Directional Light Component Object Reference)
   > SetVisibility(self=@n19, bNewVisibility=NOT Boolean(A=Get loc_light()), bPropagateToChildren=true)
   n20 = RemoveIsoJunk(execute=@n18.else)
   n21 = Get ToolEnabled()
   n22 = Branch(execute=@n17.else, Condition=AND Boolean(A=NOT Boolean(A=@n21), B=Get loc_light()))
   AddSkyDome(execute=@n22.else)
   > n23 = GetActorOfClass(ActorClass=Get Enviro Actor())
   > n24 = SetVisibility(
    self=GetComponentByClass(self=@n23, ComponentClass=&StaticMeshComponent) -> (ReturnValue:Static Mesh Component Object Reference)
    bNewVisibility=Get ToolEnabled()
    bPropagateToChildren=false
   )
   n25 = GetActorOfClass(execute=@n2.then_5)
   > SetVisibility(self=Get RootComponent(self=@n25), bNewVisibility=Get loc_light(), bPropagateToChildren=true)
  ]}
 OnMouseButtonUp_0 {nodes [OnMouseButtonUp_0!() -> (then, MyGeometry:Geometry Structure, MouseEvent:"Pointer Event Structure (by ref)") {is=title K2Node_FunctionEntry}]}
 RemoveIsoJunk {nodes [
   RemoveIsoJunk() {is=title K2Node_FunctionEntry}
   > n1 = GetAllActorsOfClass(ActorClass=Get Enviro Actor()) -> (OutActors:Array of Actor Object References)
   > n2 = For Each Loop(Array:Array of Actor Object References=@n1) -> (Array Element:Actor Object Reference)
   > DestroyActor!(execute, self:Editor Actor Subsystem Object Reference=Get EditorActorSubsystem(), ActorToDestroy:Actor Object Reference=@n2) -> (then, ReturnValue:bool)
  ]}
 RestoreSequencer {nodes [
   RestoreSequencer() {is=title K2Node_FunctionEntry}
   > n1 = GetAllLevelActors(self=Get EditorActorSubsystem())
   > n2 = For Each Loop(Array=@n1)
   n3 = GetFocusedLevelSequence(execute=@n2.Completed)
   > n4 = GetBindings(Sequence=@n3)
   > n5 = For Each Loop(Array:Array of Movie Scene Binding Proxy Structures=@n4) -> (Array Element:Movie Scene Binding Proxy Structure)
   > n6 = GetBindingID(InBinding=@n5)
   > n7 = Cast To Actor(Object="Get (a copy)"(Array=GetBoundObjects(ObjectBinding=@n6)))
   > Add Unique(TargetArray:Array of Actor Object References=Get hide_array(), NewItem:"Actor Object Reference (by ref)"=@n7)
   n8 = For Each Loop(Exec=@n5.Completed, Array:Array of Actor Object References=Get hide_array()) -> (Array Element:Actor Object Reference)
   n9 = GetAllActorsOfClass(execute=@n8.Completed, ActorClass=&AnimationBrowserStage_C) -> (OutActors:Array of Animation Browser Stage Object References)
   > n10 = For Each Loop(Array:Array of Animation Browser Stage Object References=@n9) -> (Array Element:Animation Browser Stage Object Reference)
   > SetVisibility(self=Get Cube(self=@n10), bNewVisibility=IsChecked(self=Get ShowStage_CHK()), bPropagateToChildren=false)
   Set loopedActor(execute=@n8.LoopBody, loopedActor:Actor Object Reference=@n8) -> (Output_Get:Actor Object Reference)
   n11 = Get Class Display Name(Class=Get Class(Object=Get loopedActor() -> (loopedActor:Actor Object Reference)))
   > n12 = Branch(Condition=NOT Boolean(A=Contains(SearchIn=@n11, Substring=Mannequin)))
   n13 = Get loopedActor()
   > n14 = Branch(Condition="Equal (Class)"(A=Get Class(Object=@n13), B=&AnimationBrowserStage_C))
   > SetVisibility(self=Get RootComponent(self=Get loopedActor()), bNewVisibility=NOT Boolean(A=Get ToolEnabled()), bPropagateToChildren=false)
   SetVisibility(
    execute=@n14.else
    self=Get RootComponent(self=Get loopedActor())
    bNewVisibility=NOT Boolean(A=Get ToolEnabled())
    bPropagateToChildren=true
   )
   n15 = For Each Loop(
    Exec=@n12.else
    Array:Array of Skeletal Mesh Component Object References=Get Components By Class(self=Get loopedActor(), ComponentClass=&SkeletalMeshComponent) -> (ReturnValue:Array of Skeletal Mesh Component Object References)
   ) -> (Array Element:Skeletal Mesh Component Object Reference)
   > PrintString(InString=GetDisplayName(Object=@n15))
   > SetVisibility(self=@n15, bNewVisibility=NOT Boolean(A=Get ToolEnabled()), bPropagateToChildren=false)
   n16 = CullLightsFromArray(execute=@n2.LoopBody, array_elem=@n2)
   > Append Array(TargetArray=Get hide_array(), SourceArray=Make Array("[0]"=@n16))
  ]}
 SetRenderPass {nodes [
   n1 = SetRenderPass!() -> (then, SkeletalMeshComponent:Skeletal Mesh Component Object Reference) {is=title K2Node_FunctionEntry}
   PrintString(InString=GetDisplayName(Object=@n1))
   > SetRenderInMainPass!(execute, self:Primitive Component Object Reference=@n1, bValue:bool=Get RenderMainIsSet() -> (RenderMainIsSet:bool)) -> (then)
   > SetRenderInDepthPass!(execute, self:Primitive Component Object Reference=@n1, bValue:bool=Get RenderDepthIsSet() -> (RenderDepthIsSet:bool)) -> (then)
   SetCustomDepthStencilValue!(
    execute=@n2.then
    self:Primitive Component Object Reference=@n1
    Value:int=Select(
     Option 0:int=MakeLiteralInt!(Value:int=0) -> (ReturnValue:int=0)
     Option 1:int=MakeLiteralInt(Value=1) -> (ReturnValue=0)
     Index=Get RenderDepthIsSet()
    ) -> (ReturnValue:int)
   ) -> (then)
   SetRenderInMainPass(execute=@n1.then, self=@n1, bValue=Get RenderMainIsSet())
   > n2 = SetRenderInDepthPass(self=@n1, bValue=Get RenderDepthIsSet())
  ]}
 SilhouetteSequencer {nodes [
   SilhouetteSequencer() {is=title K2Node_FunctionEntry}
   > n1 = Branch(Condition=Get ToolEnabled())
   > Set RenderDepthIsSet(RenderDepthIsSet:bool=true) -> (Output_Get:bool)
   > n2 = Set RenderMainIsSet(RenderMainIsSet:bool=false) -> (Output_Get:bool)
   n3 = GetFocusedLevelSequence(execute=[@n10.then, @n2.then])
   > n4 = GetBindings(Sequence=@n3)
   > n5 = For Each Loop(Array:Array of Movie Scene Binding Proxy Structures=@n4) -> (Array Element:Movie Scene Binding Proxy Structure)
   > n6 = GetBindingID(InBinding=@n5)
   > n7 = For Each Loop(Array:Array of Object References=GetBoundObjects(ObjectBinding=@n6)) -> (Array Element:Object Reference)
   > n8 = Cast To Actor(Object=@n7)
   > n9 = For Each Loop(Array:Array of Skeletal Mesh Component Object References=Get Components By Class(self=@n8, ComponentClass=&SkeletalMeshComponent)) -> (Array Element:Skeletal Mesh Component Object Reference)
   > SetRenderPass!(execute, self, SkeletalMeshComponent:Skeletal Mesh Component Object Reference=@n9) -> (then) {is=title K2Node_CallFunction}
   Set RenderDepthIsSet(execute=@n1.else, RenderDepthIsSet=false)
   > n10 = Set RenderMainIsSet(RenderMainIsSet=true)
  ]}
]
variables [
 BackgroundColor {type=Color, default=<rgba 0.2275 0.2275 0.2275 0>}
 BGColor {type=LinearColor, default=<rgba 1 0 0 1>}
 ControlRig {type=ControlRig, default=null}
 Enviro Actor {type=Class, default=&BP_IsolateEnviro_C}
 EnvSphere {type=Actor, default=null}
 GridMID {type=MaterialInstanceDynamic}
 hide_array {type="array<Actor>"}
 hideStage {type=boolean}
 IsolatedActors {type="array<Actor>"}
 LSActors_array {type="array<Actor>"}
 OnlyOwnerSee {type=boolean}
 RenderDepthIsSet {type=boolean}
 RenderMainIsSet {type=boolean, default=true}
 rendertarget {type=Texture2D}
 selectedActors_array {type="array<Actor>"}
 SelectedActorsTransform {type=Vector}
 SilhouetteColor {type=Color, default=<rgba 0 0 0 0>}
 SilhouetteOn {type=boolean}
 SilhouetteParameter {type=string, default=SilhouetteColor}
 ToolEnabled {type=boolean}
]
widgets {
 info {rootWidgetClass=&ScaleBox, widgetCount=41}
 widgets [
  ScaleBox_107 {widgetClassPath=&ScaleBox, widget=ScaleBox_107}
  VerticalBox_0 {
   widgetClassPath=&VerticalBox
   parent=ScaleBox_107
   widget=VerticalBox_0
   slot {class=ScaleBoxSlot, properties {verticalAlignment=VAlign_Top}}
  }
  HorizontalBox_17 {
   widgetClassPath=&HorizontalBox
   parent=VerticalBox_0
   widget=HorizontalBox_17
   slot {class=VerticalBoxSlot, properties {padding=<ltrb 0 2 0 0>}}
  }
  VerticalBox_2 {
   widgetClassPath=&VerticalBox
   parent=HorizontalBox_17
   widget=VerticalBox_2
   slot {class=HorizontalBoxSlot, properties {verticalAlignment=VAlign_Center, size {sizeRule=Fill, value=1}}}
  }
  SilhouetteMode_BTN {
   widgetClassPath=&EditorUtilityButton
   parent=VerticalBox_2
   widget=SilhouetteMode_BTN
   properties {
    toolTipText=Silhouette everything in your sequencer
    widgetStyle {
     clickedSlateSound {resourceObject=null}
     disabled {
      bIsDynamicallyLoaded=false
      drawAs=RoundedBox
      imageType=NoImage
      mirroring=NoMirror
      resourceName=None
      resourceObject=null
      tiling=NoTile
      imageSize=<0 0>
      margin=<ltrb 0 0 0 0>
      outlineSettings {
       bUseBrushTransparency=false
       roundingType=FixedRadius
       width=1
       color {colorUseRule=UseColor_Specified, specifiedColor=<rgba 0.0395 0.0395 0.0395 1>}
       cornerRadii=<4 4 4 4>
      }
      tintColor {colorUseRule=UseColor_Specified, specifiedColor=<rgba 0.0395 0.0395 0.0395 1>}
      uVRegion {bIsValid=false, max=<0 0>, min=<0 0>}
     }
     disabledForeground {colorUseRule=UseColor_ColorTable, specifiedColor=<rgba 0 0 0 0>}
     hovered {
      bIsDynamicallyLoaded=false
      drawAs=RoundedBox
      imageType=NoImage
      mirroring=NoMirror
      resourceName=None
      resourceObject=null
      tiling=NoTile
      imageSize=<32 32>
      margin=<ltrb 0 0 0 0>
      outlineSettings {
       bUseBrushTransparency=false
       roundingType=FixedRadius
       width=1
       color {colorUseRule=UseColor_Specified, specifiedColor=<rgba 0.0395 0.0395 0.0395 1>}
       cornerRadii=<4 4 4 4>
      }
      tintColor {colorUseRule=UseColor_Specified, specifiedColor=<rgba 0.0953 0.0953 0.0953 1>}
      uVRegion {bIsValid=false, max=<0 0>, min=<0 0>}
     }
     hoveredForeground {colorUseRule=UseColor_ColorTable, specifiedColor=<rgba 0 0 0 0>}
     hoveredSlateSound {resourceObject=null}
     normal {
      bIsDynamicallyLoaded=false
      drawAs=RoundedBox
      imageType=NoImage
      mirroring=NoMirror
      resourceName=None
      resourceObject=null
      tiling=NoTile
      imageSize=<32 32>
      margin=<ltrb 0 0 0 0>
      outlineSettings {
       bUseBrushTransparency=false
       roundingType=FixedRadius
       width=1
       color {colorUseRule=UseColor_Specified, specifiedColor=<rgba 0.0395 0.0395 0.0395 1>}
       cornerRadii=<4 4 4 4>
      }
      tintColor {colorUseRule=UseColor_Specified, specifiedColor=<rgba 0.0395 0.0395 0.0395 1>}
      uVRegion {bIsValid=false, max=<0 0>, min=<0 0>}
     }
     normalForeground {colorUseRule=UseColor_ColorTable, specifiedColor=<rgba 0 0 0 0>}
     normalPadding=<ltrb 12 1.5 12 1.5>
     pressed {
      bIsDynamicallyLoaded=false
      drawAs=RoundedBox
      imageType=NoImage
      mirroring=NoMirror
      resourceName=None
      resourceObject=null
      tiling=NoTile
      imageSize=<32 32>
      margin=<ltrb 0 0 0 0>
      outlineSettings {
       bUseBrushTransparency=false
       roundingType=FixedRadius
       width=1
       color {colorUseRule=UseColor_Specified, specifiedColor=<rgba 0.0395 0.0395 0.0395 1>}
       cornerRadii=<4 4 4 4>
      }
      tintColor {colorUseRule=UseColor_Specified, specifiedColor=<rgba 0.0284 0.0284 0.0284 1>}
      uVRegion {bIsValid=false, max=<0 0>, min=<0 0>}
     }
     pressedForeground {colorUseRule=UseColor_ColorTable, specifiedColor=<rgba 0 0 0 0>}
     pressedPadding=<ltrb 12 2.5 12 0.5>
     pressedSlateSound {resourceObject=null}
    }
   }
   slot {class=VerticalBoxSlot, properties {verticalAlignment=VAlign_Center, padding=<ltrb 2 2 2 2>}}
  }
  HorizontalBox_0 {
   widgetClassPath=&HorizontalBox
   parent=SilhouetteMode_BTN
   widget=HorizontalBox_0
   properties {bOverride_Cursor=true}
   slot {class=ButtonSlot, properties {horizontalAlignment=HAlign_Left, padding=<ltrb 0 2 0 2>}}
  }
  Image_118 {
   widgetClassPath=&Image
   bIsVariable=true
   parent=HorizontalBox_0
   widget=Image_118
   properties {brush {
     bIsDynamicallyLoaded=false
     drawAs=Image
     imageType=FullColor
     mirroring=NoMirror
     resourceName=None
     resourceObject=/AnimInEdTools/AnimationToolbox/Icons/AnimToolbox_UI_Icon_21
     tiling=NoTile
     imageSize=<20 20>
     margin=<ltrb 0 0 0 0>
     outlineSettings {
      bUseBrushTransparency=false
      roundingType=HalfHeightRadius
      width=0
      color {colorUseRule=UseColor_Specified, specifiedColor=<rgba 0 0 0 0>}
      cornerRadii=<0 0 0 0>
     }
     tintColor {colorUseRule=UseColor_Specified, specifiedColor=<rgba 1 1 1 1>}
     uVRegion {bIsValid=false, max=<0 0>, min=<0 0>}
    }}
   slot {class=HorizontalBoxSlot, properties {horizontalAlignment=HAlign_Center, verticalAlignment=VAlign_Center}}
  }
  TextBlock_0 {
   widgetClassPath=&TextBlock
   parent=HorizontalBox_0
   widget=TextBlock_0
   properties {
    text=Silhouette
    font {
     bForceMonospaced=false
     bMaterialIsStencil=false
     fontMaterial=null
     fontObject=&Roboto
     letterSpacing=0
     monospacedWidth=1
     size=12
     skewAmount=0
     typefaceFontName=Regular
     outlineSettings {
      bApplyOutlineToDropShadows=false
      bMiteredCorners=false
      bSeparateFillAlpha=false
      outlineMaterial=null
      outlineSize=0
      outlineColor=<rgba 0 0 0 1>
     }
    }
   }
   slot {class=HorizontalBoxSlot, properties {horizontalAlignment=HAlign_Center, verticalAlignment=VAlign_Center, padding=<ltrb 3 0 0 0>}}
  }
  Isolate_BTN {
   widgetClassPath=&EditorUtilityButton
   parent=VerticalBox_2
   widget=Isolate_BTN
   properties {
    toolTipText="Isolate based on selection. If nothing selected, the contents of the Sequencer will be Isolated"
    widgetStyle {
     clickedSlateSound {resourceObject=null}
     disabled {
      bIsDynamicallyLoaded=false
      drawAs=RoundedBox
      imageType=NoImage
      mirroring=NoMirror
      resourceName=None
      resourceObject=null
      tiling=NoTile
      imageSize=<0 0>
      margin=<ltrb 0 0 0 0>
      outlineSettings {
       bUseBrushTransparency=false
       roundingType=FixedRadius
       width=1
       color {colorUseRule=UseColor_Specified, specifiedColor=<rgba 0.0395 0.0395 0.0395 1>}
       cornerRadii=<4 4 4 4>
      }
      tintColor {colorUseRule=UseColor_Specified, specifiedColor=<rgba 0.0395 0.0395 0.0395 1>}
      uVRegion {bIsValid=false, max=<0 0>, min=<0 0>}
     }
     disabledForeground {colorUseRule=UseColor_ColorTable, specifiedColor=<rgba 0 0 0 0>}
     hovered {
      bIsDynamicallyLoaded=false
      drawAs=RoundedBox
      imageType=NoImage
      mirroring=NoMirror
      resourceName=None
      resourceObject=null
      tiling=NoTile
      imageSize=<32 32>
      margin=<ltrb 0 0 0 0>
      outlineSettings {
       bUseBrushTransparency=false
       roundingType=FixedRadius
       width=1
       color {colorUseRule=UseColor_Specified, specifiedColor=<rgba 0.0395 0.0395 0.0395 1>}
       cornerRadii=<4 4 4 4>
      }
      tintColor {colorUseRule=UseColor_Specified, specifiedColor=<rgba 0.0953 0.0953 0.0953 1>}
      uVRegion {bIsValid=false, max=<0 0>, min=<0 0>}
     }
     hoveredForeground {colorUseRule=UseColor_ColorTable, specifiedColor=<rgba 0 0 0 0>}
     hoveredSlateSound {resourceObject=null}
     normal {
      bIsDynamicallyLoaded=false
      drawAs=RoundedBox
      imageType=NoImage
      mirroring=NoMirror
      resourceName=None
      resourceObject=null
      tiling=NoTile
      imageSize=<32 32>
      margin=<ltrb 0 0 0 0>
      outlineSettings {
       bUseBrushTransparency=false
       roundingType=FixedRadius
       width=1
       color {colorUseRule=UseColor_Specified, specifiedColor=<rgba 0.0395 0.0395 0.0395 1>}
       cornerRadii=<4 4 4 4>
      }
      tintColor {colorUseRule=UseColor_Specified, specifiedColor=<rgba 0.0395 0.0395 0.0395 1>}
      uVRegion {bIsValid=false, max=<0 0>, min=<0 0>}
     }
     normalForeground {colorUseRule=UseColor_ColorTable, specifiedColor=<rgba 0 0 0 0>}
     normalPadding=<ltrb 12 1.5 12 1.5>
     pressed {
      bIsDynamicallyLoaded=false
      drawAs=RoundedBox
      imageType=NoImage
      mirroring=NoMirror
      resourceName=None
      resourceObject=null
      tiling=NoTile
      imageSize=<32 32>
      margin=<ltrb 0 0 0 0>
      outlineSettings {
       bUseBrushTransparency=false
       roundingType=FixedRadius
       width=1
       color {colorUseRule=UseColor_Specified, specifiedColor=<rgba 0.0395 0.0395 0.0395 1>}
       cornerRadii=<4 4 4 4>
      }
      tintColor {colorUseRule=UseColor_Specified, specifiedColor=<rgba 0.0284 0.0284 0.0284 1>}
      uVRegion {bIsValid=false, max=<0 0>, min=<0 0>}
     }
     pressedForeground {colorUseRule=UseColor_ColorTable, specifiedColor=<rgba 0 0 0 0>}
     pressedPadding=<ltrb 12 2.5 12 0.5>
     pressedSlateSound {resourceObject=null}
    }
   }
   slot {class=VerticalBoxSlot, properties {verticalAlignment=VAlign_Center, padding=<ltrb 2 2 2 2>}}
  }
  HorizontalBox_1 {
   widgetClassPath=&HorizontalBox
   parent=Isolate_BTN
   widget=HorizontalBox_1
   properties {bOverride_Cursor=true}
   slot {class=ButtonSlot, properties {horizontalAlignment=HAlign_Left, padding=<ltrb 0 2 0 2>}}
  }
  Image_240 {
   widgetClassPath=&Image
   bIsVariable=true
   parent=HorizontalBox_1
   widget=Image_240
   properties {brush {
     bIsDynamicallyLoaded=false
     drawAs=Image
     imageType=FullColor
     mirroring=NoMirror
     resourceName=None
     resourceObject=/AnimInEdTools/AnimationToolbox/Icons/AnimToolbox_UI_Icon_19
     tiling=NoTile
     imageSize=<20 20>
     margin=<ltrb 0 0 0 0>
     outlineSettings {
      bUseBrushTransparency=false
      roundingType=HalfHeightRadius
      width=0
      color {colorUseRule=UseColor_Specified, specifiedColor=<rgba 0 0 0 0>}
      cornerRadii=<0 0 0 0>
     }
     tintColor {colorUseRule=UseColor_Specified, specifiedColor=<rgba 1 1 1 1>}
     uVRegion {bIsValid=false, max=<0 0>, min=<0 0>}
    }}
   slot {class=HorizontalBoxSlot, properties {horizontalAlignment=HAlign_Center, verticalAlignment=VAlign_Center}}
  }
  TextBlock_29 {
   widgetClassPath=&TextBlock
   parent=HorizontalBox_1
   widget=TextBlock_29
   properties {
    text=Isolate
    font {
     bForceMonospaced=false
     bMaterialIsStencil=false
     fontMaterial=null
     fontObject=&Roboto
     letterSpacing=0
     monospacedWidth=1
     size=12
     skewAmount=0
     typefaceFontName=Regular
     outlineSettings {
      bApplyOutlineToDropShadows=false
      bMiteredCorners=false
      bSeparateFillAlpha=false
      outlineMaterial=null
      outlineSize=0
      outlineColor=<rgba 0 0 0 1>
     }
    }
   }
   slot {class=HorizontalBoxSlot, properties {horizontalAlignment=HAlign_Center, verticalAlignment=VAlign_Center, padding=<ltrb 3 0 0 0>}}
  }
  SizeBox_1 {
   widgetClassPath=&SizeBox
   parent=HorizontalBox_17
   widget=SizeBox_1
   properties {bOverride_MaxDesiredWidth=true, maxDesiredWidth=25}
   slot {class=HorizontalBoxSlot, properties {padding=<ltrb 2 2 2 2>}}
  }
  Reset_BTN {
   widgetClassPath=&EditorUtilityButton
   parent=SizeBox_1
   widget=Reset_BTN
   properties {
    toolTipText=Reset Silhouette/Isolate Mode
    widgetStyle {
     clickedSlateSound {resourceObject=null}
     disabled {
      bIsDynamicallyLoaded=false
      drawAs=RoundedBox
      imageType=NoImage
      mirroring=NoMirror
      resourceName=None
      resourceObject=null
      tiling=NoTile
      imageSize=<0 0>
      margin=<ltrb 0 0 0 0>
      outlineSettings {
       bUseBrushTransparency=false
       roundingType=FixedRadius
       width=1
       color {colorUseRule=UseColor_Specified, specifiedColor=<rgba 0.0395 0.0395 0.0395 1>}
       cornerRadii=<4 4 4 4>
      }
      tintColor {colorUseRule=UseColor_Specified, specifiedColor=<rgba 0.0395 0.0395 0.0395 1>}
      uVRegion {bIsValid=false, max=<0 0>, min=<0 0>}
     }
     disabledForeground {colorUseRule=UseColor_ColorTable, specifiedColor=<rgba 0 0 0 0>}
     hovered {
      bIsDynamicallyLoaded=false
      drawAs=RoundedBox
      imageType=NoImage
      mirroring=NoMirror
      resourceName=None
      resourceObject=null
      tiling=NoTile
      imageSize=<32 32>
      margin=<ltrb 0 0 0 0>
      outlineSettings {
       bUseBrushTransparency=false
       roundingType=FixedRadius
       width=1
       color {colorUseRule=UseColor_Specified, specifiedColor=<rgba 0.0395 0.0395 0.0395 1>}
       cornerRadii=<4 4 4 4>
      }
      tintColor {colorUseRule=UseColor_Specified, specifiedColor=<rgba 0.0953 0.0953 0.0953 1>}
      uVRegion {bIsValid=false, max=<0 0>, min=<0 0>}
     }
     hoveredForeground {colorUseRule=UseColor_ColorTable, specifiedColor=<rgba 0 0 0 0>}
     hoveredSlateSound {resourceObject=null}
     normal {
      bIsDynamicallyLoaded=false
      drawAs=RoundedBox
      imageType=NoImage
      mirroring=NoMirror
      resourceName=None
      resourceObject=null
      tiling=NoTile
      imageSize=<32 32>
      margin=<ltrb 0 0 0 0>
      outlineSettings {
       bUseBrushTransparency=false
       roundingType=FixedRadius
       width=1
       color {colorUseRule=UseColor_Specified, specifiedColor=<rgba 0.0395 0.0395 0.0395 1>}
       cornerRadii=<4 4 4 4>
      }
      tintColor {colorUseRule=UseColor_Specified, specifiedColor=<rgba 0.0395 0.0395 0.0395 1>}
      uVRegion {bIsValid=false, max=<0 0>, min=<0 0>}
     }
     normalForeground {colorUseRule=UseColor_ColorTable, specifiedColor=<rgba 0 0 0 0>}
     normalPadding=<ltrb 12 1.5 12 1.5>
     pressed {
      bIsDynamicallyLoaded=false
      drawAs=RoundedBox
      imageType=NoImage
      mirroring=NoMirror
      resourceName=None
      resourceObject=null
      tiling=NoTile
      imageSize=<32 32>
      margin=<ltrb 0 0 0 0>
      outlineSettings {
       bUseBrushTransparency=false
       roundingType=FixedRadius
       width=1
       color {colorUseRule=UseColor_Specified, specifiedColor=<rgba 0.0395 0.0395 0.0395 1>}
       cornerRadii=<4 4 4 4>
      }
      tintColor {colorUseRule=UseColor_Specified, specifiedColor=<rgba 0.0284 0.0284 0.0284 1>}
      uVRegion {bIsValid=false, max=<0 0>, min=<0 0>}
     }
     pressedForeground {colorUseRule=UseColor_ColorTable, specifiedColor=<rgba 0 0 0 0>}
     pressedPadding=<ltrb 12 2.5 12 0.5>
     pressedSlateSound {resourceObject=null}
    }
   }
   slot {class=SizeBoxSlot, properties {horizontalAlignment=HAlign_Center}}
  }
  ScaleBox_1 {
   widgetClassPath=&ScaleBox
   parent=Reset_BTN
   widget=ScaleBox_1
   properties {stretch=None, userSpecifiedScale=10.8185}
   slot {class=ButtonSlot, properties {padding=<ltrb 0 0 0 0>}}
  }
  VerticalBox_4 {widgetClassPath=&VerticalBox, parent=ScaleBox_1, widget=VerticalBox_4, slot {class=ScaleBoxSlot}}
  Image_330 {
   widgetClassPath=&Image
   bIsVariable=true
   parent=VerticalBox_4
   widget=Image_330
   properties {brush {
     bIsDynamicallyLoaded=false
     drawAs=Image
     imageType=FullColor
     mirroring=NoMirror
     resourceName=None
     resourceObject=/AnimInEdTools/Pickwalk_ControlRig/icons/AnimToolbox_UI_Pickwalk_Back
     tiling=NoTile
     imageSize=<20 20>
     margin=<ltrb 0 0 0 0>
     outlineSettings {
      bUseBrushTransparency=false
      roundingType=HalfHeightRadius
      width=0
      color {colorUseRule=UseColor_Specified, specifiedColor=<rgba 0 0 0 0>}
      cornerRadii=<0 0 0 0>
     }
     tintColor {colorUseRule=UseColor_Specified, specifiedColor=<rgba 1 1 1 1>}
     uVRegion {bIsValid=false, max=<0 0>, min=<0 0>}
    }}
   slot {class=VerticalBoxSlot}
  }
  TextBlock_1 {
   widgetClassPath=&TextBlock
   parent=VerticalBox_4
   widget=TextBlock_1
   properties {
    text=Reset
    font {
     bForceMonospaced=false
     bMaterialIsStencil=false
     fontMaterial=null
     fontObject=&Roboto
     letterSpacing=0
     monospacedWidth=1
     size=4.52
     skewAmount=0
     typefaceFontName=Bold
     outlineSettings {
      bApplyOutlineToDropShadows=false
      bMiteredCorners=false
      bSeparateFillAlpha=false
      outlineMaterial=null
      outlineSize=0
      outlineColor=<rgba 0 0 0 1>
     }
    }
   }
   slot {class=VerticalBoxSlot, properties {horizontalAlignment=HAlign_Center}}
  }
  HorizontalBox_3 {
   widgetClassPath=&HorizontalBox
   parent=VerticalBox_0
   widget=HorizontalBox_3
   slot {class=VerticalBoxSlot, properties {horizontalAlignment=HAlign_Center, verticalAlignment=VAlign_Center}}
  }
  ColorBG {
   widgetClassPath=&SinglePropertyView
   parent=HorizontalBox_3
   widget=ColorBG
   properties {propertyName=BackgroundColor}
   slot {class=HorizontalBoxSlot, properties {horizontalAlignment=HAlign_Center}}
  }
  Border_88 {
   widgetClassPath=&Border
   parent=VerticalBox_0
   widget=Border_88
   properties {brushColor=<rgba 0.0395 0.0395 0.0395 1>, contentColorAndOpacity=<rgba 0.0395 0.0395 0.0395 1>, padding=<ltrb 4 1 4 0>}
   slot {class=VerticalBoxSlot, properties {padding=<ltrb 0 2 0 0>}}
  }
  HorizontalBox_2 {
   widgetClassPath=&HorizontalBox
   parent=VerticalBox_0
   widget=HorizontalBox_2
   slot {class=VerticalBoxSlot, properties {padding=<ltrb 1 0 1 1>, size {sizeRule=Fill, value=1}}}
  }
  HBox_Lights {widgetClassPath=&HorizontalBox, parent=HorizontalBox_2, widget=HBox_Lights, slot {class=HorizontalBoxSlot}}
  UEFN_TextBlock_C_110 {
   widgetClassPath=&UEFN_TextBlock
   parent=HBox_Lights
   widget=UEFN_TextBlock_C_110
   properties {
    text=Scene Lights:
    font {
     bForceMonospaced=false
     bMaterialIsStencil=false
     fontMaterial=null
     fontObject=&Roboto
     letterSpacing=0
     monospacedWidth=1
     size=6.75
     skewAmount=0
     typefaceFontName=Regular
     outlineSettings {
      bApplyOutlineToDropShadows=false
      bMiteredCorners=false
      bSeparateFillAlpha=false
      outlineMaterial=null
      outlineSize=0
      outlineColor=<rgba 0 0 0 1>
     }
    }
   }
   slot {class=HorizontalBoxSlot, properties {horizontalAlignment=HAlign_Center, verticalAlignment=VAlign_Top, padding=<ltrb 0 3 0 0>}}
  }
  VerticalBox_1 {widgetClassPath=&VerticalBox, parent=HBox_Lights, widget=VerticalBox_1, slot {class=HorizontalBoxSlot}}
  HorizontalBox {widgetClassPath=&HorizontalBox, parent=VerticalBox_1, widget=HorizontalBox, slot {class=VerticalBoxSlot}}
  TextBlock {
   widgetClassPath=&TextBlock
   parent=HorizontalBox
   widget=TextBlock
   properties {
    text=On/Off
    font {
     bForceMonospaced=false
     bMaterialIsStencil=false
     fontMaterial=null
     fontObject=&Roboto
     letterSpacing=0
     monospacedWidth=1
     size=4.5
     skewAmount=0
     typefaceFontName=Regular
     outlineSettings {
      bApplyOutlineToDropShadows=false
      bMiteredCorners=false
      bSeparateFillAlpha=false
      outlineMaterial=null
      outlineSize=0
      outlineColor=<rgba 0 0 0 1>
     }
    }
   }
   slot {
    class=HorizontalBoxSlot
    properties {horizontalAlignment=HAlign_Left, verticalAlignment=VAlign_Center, padding=<ltrb 4 0 0 0>, size {sizeRule=Fill, value=1}}
   }
  }
  HideLight_CHK {
   widgetClassPath=&EditorUtilityCheckBox
   bIsVariable=true
   parent=HorizontalBox
   widget=HideLight_CHK
   properties {
    checkedState=Checked
    horizontalAlignment=HAlign_Left
    renderTransform {angle=0, scale=<1 0.5>, shear=<0 0>, translation=<0 0>}
    widgetStyle {
     checkBoxType=CheckBox
     backgroundHoveredImage {
      bIsDynamicallyLoaded=false
      drawAs=Box
      imageType=NoImage
      mirroring=NoMirror
      resourceName=None
      resourceObject=null
      tiling=NoTile
      imageSize=<18 18>
      margin=<ltrb 0 0 0 0>
      outlineSettings {
       bUseBrushTransparency=false
       roundingType=FixedRadius
       width=1
       color {colorUseRule=UseColor_ColorTable, specifiedColor=<rgba 0 0 0 0>}
       cornerRadii=<3 3 3 3>
      }
      tintColor {colorUseRule=UseColor_Specified, specifiedColor=<rgba 0.0048 0.0048 0.0048 0>}
      uVRegion {bIsValid=false, max=<0 0>, min=<0 0>}
     }
     backgroundImage {
      bIsDynamicallyLoaded=false
      drawAs=RoundedBox
      imageType=NoImage
      mirroring=NoMirror
      resourceName=None
      resourceObject=null
      tiling=NoTile
      imageSize=<18 18>
      margin=<ltrb 0 0 0 0>
      outlineSettings {
       bUseBrushTransparency=false
       roundingType=FixedRadius
       width=0
       color {colorUseRule=UseColor_ColorTable, specifiedColor=<rgba 0 0 0 0>}
       cornerRadii=<3 3 3 3>
      }
      tintColor {colorUseRule=UseColor_Specified, specifiedColor=<rgba 0.0048 0.0048 0.0048 0>}
      uVRegion {bIsValid=false, max=<0 0>, min=<0 0>}
     }
     backgroundPressedImage {
      bIsDynamicallyLoaded=false
      drawAs=RoundedBox
      imageType=NoImage
      mirroring=NoMirror
      resourceName=None
      resourceObject=null
      tiling=NoTile
      imageSize=<18 18>
      margin=<ltrb 0 0 0 0>
      outlineSettings {
       bUseBrushTransparency=false
       roundingType=FixedRadius
       width=0
       color {colorUseRule=UseColor_ColorTable, specifiedColor=<rgba 0 0 0 0>}
       cornerRadii=<3 3 3 3>
      }
      tintColor {colorUseRule=UseColor_Specified, specifiedColor=<rgba 0.0048 0.0048 0.0048 0>}
      uVRegion {bIsValid=false, max=<0 0>, min=<0 0>}
     }
     borderBackgroundColor {colorUseRule=UseColor_Specified, specifiedColor=<rgba 1 1 1 0>}
     checkedForeground {colorUseRule=UseColor_Specified, specifiedColor=<rgba 1 1 1 1>}
     checkedHoveredForeground {colorUseRule=UseColor_Specified, specifiedColor=<rgba 1 1 1 1>}
     checkedHoveredImage {
      bIsDynamicallyLoaded=false
      drawAs=Image
      imageType=FullColor
      mirroring=NoMirror
      resourceName=None
      resourceObject=/AnimInEdTools/AnimationToolbox/Icons/AnimToolbox_UI_Icon_toggle_on
      tiling=NoTile
      imageSize=<8 4>
      margin=<ltrb 0 0 0 0>
      outlineSettings {
       bUseBrushTransparency=false
       roundingType=HalfHeightRadius
       width=0
       color {colorUseRule=UseColor_Specified, specifiedColor=<rgba 0 0 0 0>}
       cornerRadii=<0 0 0 0>
      }
      tintColor {colorUseRule=UseColor_Specified, specifiedColor=<rgba 1 1 1 1>}
      uVRegion {bIsValid=false, max=<0 0>, min=<0 0>}
     }
     checkedImage {
      bIsDynamicallyLoaded=false
      drawAs=Image
      imageType=FullColor
      mirroring=NoMirror
      resourceName=None
      resourceObject=&AnimToolbox_UI_Icon_toggle_on
      tiling=NoTile
      imageSize=<8 4>
      margin=<ltrb 0 0 0 0>
      outlineSettings {
       bUseBrushTransparency=false
       roundingType=HalfHeightRadius
       width=0
       color {colorUseRule=UseColor_Specified, specifiedColor=<rgba 0 0 0 0>}
       cornerRadii=<0 0 0 0>
      }
      tintColor {colorUseRule=UseColor_Specified, specifiedColor=<rgba 0.6658 0.6658 0.6658 1>}
      uVRegion {bIsValid=false, max=<0 0>, min=<0 0>}
     }
     checkedPressedForeground {colorUseRule=UseColor_Specified, specifiedColor=<rgba 1 1 1 1>}
     checkedPressedImage {
      bIsDynamicallyLoaded=false
      drawAs=Image
      imageType=FullColor
      mirroring=NoMirror
      resourceName=None
      resourceObject=/AnimInEdTools/AnimationToolbox/Icons/AnimToolbox_UI_Icon_toggle_off
      tiling=NoTile
      imageSize=<8 4>
      margin=<ltrb 0 0 0 0>
      outlineSettings {
       bUseBrushTransparency=false
       roundingType=HalfHeightRadius
       width=0
       color {colorUseRule=UseColor_Specified, specifiedColor=<rgba 0 0 0 0>}
       cornerRadii=<0 0 0 0>
      }
      tintColor {colorUseRule=UseColor_Specified, specifiedColor=<rgba 1 1 1 1>}
      uVRegion {bIsValid=false, max=<0 0>, min=<0 0>}
     }
     checkedSlateSound {resourceObject=null}
     foregroundColor {colorUseRule=UseColor_Specified, specifiedColor=<rgba 1 1 1 1>}
     hoveredForeground {colorUseRule=UseColor_Specified, specifiedColor=<rgba 1 1 1 1>}
     hoveredSlateSound {resourceObject=null}
     padding=<ltrb 0 0 0 0>
     pressedForeground {colorUseRule=UseColor_Specified, specifiedColor=<rgba 1 1 1 1>}
     uncheckedHoveredImage {
      bIsDynamicallyLoaded=false
      drawAs=Image
      imageType=FullColor
      mirroring=NoMirror
      resourceName=None
      resourceObject=&AnimToolbox_UI_Icon_toggle_off
      tiling=NoTile
      imageSize=<8 4>
      margin=<ltrb 0 0 0 0>
      outlineSettings {
       bUseBrushTransparency=false
       roundingType=HalfHeightRadius
       width=0
       color {colorUseRule=UseColor_Specified, specifiedColor=<rgba 0 0 0 0>}
       cornerRadii=<0 0 0 0>
      }
      tintColor {colorUseRule=UseColor_Specified, specifiedColor=<rgba 1 1 1 1>}
      uVRegion {bIsValid=false, max=<0 0>, min=<0 0>}
     }
     uncheckedImage {
      bIsDynamicallyLoaded=false
      drawAs=Image
      imageType=FullColor
      mirroring=NoMirror
      resourceName=None
      resourceObject=&AnimToolbox_UI_Icon_toggle_off
      tiling=NoTile
      imageSize=<8 4>
      margin=<ltrb 0 0 0 0>
      outlineSettings {
       bUseBrushTransparency=false
       roundingType=HalfHeightRadius
       width=0
       color {colorUseRule=UseColor_Specified, specifiedColor=<rgba 0 0 0 0>}
       cornerRadii=<0 0 0 0>
      }
      tintColor {colorUseRule=UseColor_Specified, specifiedColor=<rgba 0.666 0.666 0.666 1>}
      uVRegion {bIsValid=false, max=<0 0>, min=<0 0>}
     }
     uncheckedPressedImage {
      bIsDynamicallyLoaded=false
      drawAs=Image
      imageType=FullColor
      mirroring=NoMirror
      resourceName=None
      resourceObject=&AnimToolbox_UI_Icon_toggle_off
      tiling=NoTile
      imageSize=<8 4>
      margin=<ltrb 0 0 0 0>
      outlineSettings {
       bUseBrushTransparency=false
       roundingType=HalfHeightRadius
       width=0
       color {colorUseRule=UseColor_Specified, specifiedColor=<rgba 0 0 0 0>}
       cornerRadii=<0 0 0 0>
      }
      tintColor {colorUseRule=UseColor_Specified, specifiedColor=<rgba 1 1 1 1>}
      uVRegion {bIsValid=false, max=<0 0>, min=<0 0>}
     }
     uncheckedSlateSound {resourceObject=null}
     undeterminedForeground {colorUseRule=UseColor_Specified, specifiedColor=<rgba 1 1 1 1>}
     undeterminedHoveredImage {
      bIsDynamicallyLoaded=false
      drawAs=Image
      imageType=Vector
      mirroring=NoMirror
      resourceName=../../../Engine/Content/Slate/Starship/CoreWidgets/CheckBox/indeterminate.svg
      resourceObject=null
      tiling=NoTile
      imageSize=<16 16>
      margin=<ltrb 0 0 0 0>
      outlineSettings {
       bUseBrushTransparency=false
       roundingType=HalfHeightRadius
       width=0
       color {colorUseRule=UseColor_Specified, specifiedColor=<rgba 0 0 0 0>}
       cornerRadii=<0 0 0 0>
      }
      tintColor {colorUseRule=UseColor_Specified, specifiedColor=<rgba 0.0044 0.2384 1 1>}
      uVRegion {bIsValid=false, max=<0 0>, min=<0 0>}
     }
     undeterminedImage {
      bIsDynamicallyLoaded=false
      drawAs=Image
      imageType=Vector
      mirroring=NoMirror
      resourceName=../../../Engine/Content/Slate/Starship/CoreWidgets/CheckBox/indeterminate.svg
      resourceObject=null
      tiling=NoTile
      imageSize=<16 16>
      margin=<ltrb 0 0 0 0>
      outlineSettings {
       bUseBrushTransparency=false
       roundingType=HalfHeightRadius
       width=0
       color {colorUseRule=UseColor_Specified, specifiedColor=<rgba 0 0 0 0>}
       cornerRadii=<0 0 0 0>
      }
      tintColor {colorUseRule=UseColor_Specified, specifiedColor=<rgba 0 0.162 0.7454 1>}
      uVRegion {bIsValid=false, max=<0 0>, min=<0 0>}
     }
     undeterminedPressedImage {
      bIsDynamicallyLoaded=false
      drawAs=Image
      imageType=Vector
      mirroring=NoMirror
      resourceName=../../../Engine/Content/Slate/Starship/CoreWidgets/CheckBox/indeterminate.svg
      resourceObject=null
      tiling=NoTile
      imageSize=<16 16>
      margin=<ltrb 0 0 0 0>
      outlineSettings {
       bUseBrushTransparency=false
       roundingType=HalfHeightRadius
       width=0
       color {colorUseRule=UseColor_Specified, specifiedColor=<rgba 0 0 0 0>}
       cornerRadii=<0 0 0 0>
      }
      tintColor {colorUseRule=UseColor_Specified, specifiedColor=<rgba 0 0.0802 0.3515 1>}
      uVRegion {bIsValid=false, max=<0 0>, min=<0 0>}
     }
    }
   }
   slot {class=HorizontalBoxSlot, properties {horizontalAlignment=HAlign_Right, verticalAlignment=VAlign_Center}}
  }
  HorizontalBox_Shadows {
   widgetClassPath=&HorizontalBox
   bIsVariable=true
   parent=VerticalBox_1
   widget=HorizontalBox_Shadows
   slot {class=VerticalBoxSlot}
  }
  TextBlock_3 {
   widgetClassPath=&TextBlock
   parent=HorizontalBox_Shadows
   widget=TextBlock_3
   properties {
    text=Shadows
    font {
     bForceMonospaced=false
     bMaterialIsStencil=false
     fontMaterial=null
     fontObject=&Roboto
     letterSpacing=0
     monospacedWidth=1
     size=4.5
     skewAmount=0
     typefaceFontName=Regular
     outlineSettings {
      bApplyOutlineToDropShadows=false
      bMiteredCorners=false
      bSeparateFillAlpha=false
      outlineMaterial=null
      outlineSize=0
      outlineColor=<rgba 0 0 0 1>
     }
    }
   }
   slot {
    class=HorizontalBoxSlot
    properties {horizontalAlignment=HAlign_Left, verticalAlignment=VAlign_Center, padding=<ltrb 4 0 0 0>, size {sizeRule=Fill, value=1}}
   }
  }
  HideShadow_CHK {
   widgetClassPath=&EditorUtilityCheckBox
   bIsVariable=true
   parent=HorizontalBox_Shadows
   widget=HideShadow_CHK
   properties {
    checkedState=Checked
    horizontalAlignment=HAlign_Left
    renderTransform {angle=0, scale=<0.75 0.75>, shear=<0 0>, translation=<0 0>}
    widgetStyle {
     checkBoxType=CheckBox
     backgroundHoveredImage {
      bIsDynamicallyLoaded=false
      drawAs=RoundedBox
      imageType=NoImage
      mirroring=NoMirror
      resourceName=None
      resourceObject=null
      tiling=NoTile
      imageSize=<18 18>
      margin=<ltrb 0 0 0 0>
      outlineSettings {
       bUseBrushTransparency=false
       roundingType=FixedRadius
       width=1
       color {colorUseRule=UseColor_ColorTable, specifiedColor=<rgba 0 0 0 0>}
       cornerRadii=<3 3 3 3>
      }
      tintColor {colorUseRule=UseColor_Specified, specifiedColor=<rgba 0.0048 0.0048 0.0048 0>}
      uVRegion {bIsValid=false, max=<0 0>, min=<0 0>}
     }
     backgroundImage {
      bIsDynamicallyLoaded=false
      drawAs=RoundedBox
      imageType=NoImage
      mirroring=NoMirror
      resourceName=None
      resourceObject=null
      tiling=NoTile
      imageSize=<18 18>
      margin=<ltrb 0 0 0 0>
      outlineSettings {
       bUseBrushTransparency=false
       roundingType=FixedRadius
       width=1
       color {colorUseRule=UseColor_ColorTable, specifiedColor=<rgba 0 0 0 0>}
       cornerRadii=<3 3 3 3>
      }
      tintColor {colorUseRule=UseColor_Specified, specifiedColor=<rgba 1 1 1 0>}
      uVRegion {bIsValid=false, max=<0 0>, min=<0 0>}
     }
     backgroundPressedImage {
      bIsDynamicallyLoaded=false
      drawAs=RoundedBox
      imageType=NoImage
      mirroring=NoMirror
      resourceName=None
      resourceObject=null
      tiling=NoTile
      imageSize=<18 18>
      margin=<ltrb 0 0 0 0>
      outlineSettings {
       bUseBrushTransparency=false
       roundingType=FixedRadius
       width=1
       color {colorUseRule=UseColor_ColorTable, specifiedColor=<rgba 0 0 0 0>}
       cornerRadii=<3 3 3 3>
      }
      tintColor {colorUseRule=UseColor_Specified, specifiedColor=<rgba 1 1 1 0>}
      uVRegion {bIsValid=false, max=<0 0>, min=<0 0>}
     }
     borderBackgroundColor {colorUseRule=UseColor_Specified, specifiedColor=<rgba 1 1 1 1>}
     checkedForeground {colorUseRule=UseColor_Specified, specifiedColor=<rgba 1 1 1 1>}
     checkedHoveredForeground {colorUseRule=UseColor_Specified, specifiedColor=<rgba 1 1 1 1>}
     checkedHoveredImage {
      bIsDynamicallyLoaded=false
      drawAs=Image
      imageType=Vector
      mirroring=NoMirror
      resourceName=../../../Engine/Content/Slate/Starship/CoreWidgets/CheckBox/check.svg
      resourceObject=null
      tiling=NoTile
      imageSize=<16 16>
      margin=<ltrb 0 0 0 0>
      outlineSettings {
       bUseBrushTransparency=false
       roundingType=HalfHeightRadius
       width=0
       color {colorUseRule=UseColor_Specified, specifiedColor=<rgba 0 0 0 0>}
       cornerRadii=<0 0 0 0>
      }
      tintColor {colorUseRule=UseColor_Specified, specifiedColor=<rgba 1 1 1 1>}
      uVRegion {bIsValid=false, max=<0 0>, min=<0 0>}
     }
     checkedImage {
      bIsDynamicallyLoaded=false
      drawAs=Image
      imageType=Vector
      mirroring=NoMirror
      resourceName=../../../Engine/Content/Slate/Starship/CoreWidgets/CheckBox/check.svg
      resourceObject=null
      tiling=NoTile
      imageSize=<16 16>
      margin=<ltrb 0 0 0 0>
      outlineSettings {
       bUseBrushTransparency=false
       roundingType=HalfHeightRadius
       width=0
       color {colorUseRule=UseColor_Specified, specifiedColor=<rgba 0 0 0 0>}
       cornerRadii=<0 0 0 0>
      }
      tintColor {colorUseRule=UseColor_Specified, specifiedColor=<rgba 0.7454 0.7454 0.7454 1>}
      uVRegion {bIsValid=false, max=<0 0>, min=<0 0>}
     }
     checkedPressedForeground {colorUseRule=UseColor_Specified, specifiedColor=<rgba 1 1 1 1>}
     checkedPressedImage {
      bIsDynamicallyLoaded=false
      drawAs=Image
      imageType=Vector
      mirroring=NoMirror
      resourceName=../../../Engine/Content/Slate/Starship/CoreWidgets/CheckBox/check.svg
      resourceObject=null
      tiling=NoTile
      imageSize=<16 16>
      margin=<ltrb 0 0 0 0>
      outlineSettings {
       bUseBrushTransparency=false
       roundingType=HalfHeightRadius
       width=0
       color {colorUseRule=UseColor_Specified, specifiedColor=<rgba 0 0 0 0>}
       cornerRadii=<0 0 0 0>
      }
      tintColor {colorUseRule=UseColor_Specified, specifiedColor=<rgba 0.3515 0.3515 0.3515 1>}
      uVRegion {bIsValid=false, max=<0 0>, min=<0 0>}
     }
     checkedSlateSound {resourceObject=null}
     foregroundColor {colorUseRule=UseColor_Specified, specifiedColor=<rgba 1 1 1 1>}
     hoveredForeground {colorUseRule=UseColor_Specified, specifiedColor=<rgba 1 1 1 1>}
     hoveredSlateSound {resourceObject=null}
     padding=<ltrb 2 0 0 0>
     pressedForeground {colorUseRule=UseColor_Specified, specifiedColor=<rgba 1 1 1 1>}
     uncheckedHoveredImage {
      bIsDynamicallyLoaded=false
      drawAs=NoDrawType
      imageType=NoImage
      mirroring=NoMirror
      resourceName=None
      resourceObject=null
      tiling=NoTile
      imageSize=<0 0>
      margin=<ltrb 0 0 0 0>
      outlineSettings {
       bUseBrushTransparency=false
       roundingType=HalfHeightRadius
       width=0
       color {colorUseRule=UseColor_Specified, specifiedColor=<rgba 0 0 0 0>}
       cornerRadii=<0 0 0 0>
      }
      tintColor {colorUseRule=UseColor_Specified, specifiedColor=<rgba 1 1 1 1>}
      uVRegion {bIsValid=false, max=<0 0>, min=<0 0>}
     }
     uncheckedImage {
      bIsDynamicallyLoaded=false
      drawAs=NoDrawType
      imageType=NoImage
      mirroring=NoMirror
      resourceName=None
      resourceObject=null
      tiling=NoTile
      imageSize=<0 0>
      margin=<ltrb 0 0 0 0>
      outlineSettings {
       bUseBrushTransparency=false
       roundingType=HalfHeightRadius
       width=0
       color {colorUseRule=UseColor_Specified, specifiedColor=<rgba 0 0 0 0>}
       cornerRadii=<0 0 0 0>
      }
      tintColor {colorUseRule=UseColor_Specified, specifiedColor=<rgba 1 1 1 1>}
      uVRegion {bIsValid=false, max=<0 0>, min=<0 0>}
     }
     uncheckedPressedImage {
      bIsDynamicallyLoaded=false
      drawAs=NoDrawType
      imageType=NoImage
      mirroring=NoMirror
      resourceName=None
      resourceObject=null
      tiling=NoTile
      imageSize=<0 0>
      margin=<ltrb 0 0 0 0>
      outlineSettings {
       bUseBrushTransparency=false
       roundingType=HalfHeightRadius
       width=0
       color {colorUseRule=UseColor_Specified, specifiedColor=<rgba 0 0 0 0>}
       cornerRadii=<0 0 0 0>
      }
      tintColor {colorUseRule=UseColor_Specified, specifiedColor=<rgba 1 1 1 1>}
      uVRegion {bIsValid=false, max=<0 0>, min=<0 0>}
     }
     uncheckedSlateSound {resourceObject=null}
     undeterminedForeground {colorUseRule=UseColor_Specified, specifiedColor=<rgba 1 1 1 1>}
     undeterminedHoveredImage {
      bIsDynamicallyLoaded=false
      drawAs=Image
      imageType=Vector
      mirroring=NoMirror
      resourceName=../../../Engine/Content/Slate/Starship/CoreWidgets/CheckBox/indeterminate.svg
      resourceObject=null
      tiling=NoTile
      imageSize=<16 16>
      margin=<ltrb 0 0 0 0>
      outlineSettings {
       bUseBrushTransparency=false
       roundingType=HalfHeightRadius
       width=0
       color {colorUseRule=UseColor_Specified, specifiedColor=<rgba 0 0 0 0>}
       cornerRadii=<0 0 0 0>
      }
      tintColor {colorUseRule=UseColor_Specified, specifiedColor=<rgba 0.0044 0.2384 1 1>}
      uVRegion {bIsValid=false, max=<0 0>, min=<0 0>}
     }
     undeterminedImage {
      bIsDynamicallyLoaded=false
      drawAs=Image
      imageType=Vector
      mirroring=NoMirror
      resourceName=../../../Engine/Content/Slate/Starship/CoreWidgets/CheckBox/indeterminate.svg
      resourceObject=null
      tiling=NoTile
      imageSize=<16 16>
      margin=<ltrb 0 0 0 0>
      outlineSettings {
       bUseBrushTransparency=false
       roundingType=HalfHeightRadius
       width=0
       color {colorUseRule=UseColor_Specified, specifiedColor=<rgba 0 0 0 0>}
       cornerRadii=<0 0 0 0>
      }
      tintColor {colorUseRule=UseColor_Specified, specifiedColor=<rgba 0 0.162 0.7454 1>}
      uVRegion {bIsValid=false, max=<0 0>, min=<0 0>}
     }
     undeterminedPressedImage {
      bIsDynamicallyLoaded=false
      drawAs=Image
      imageType=Vector
      mirroring=NoMirror
      resourceName=../../../Engine/Content/Slate/Starship/CoreWidgets/CheckBox/indeterminate.svg
      resourceObject=null
      tiling=NoTile
      imageSize=<16 16>
      margin=<ltrb 0 0 0 0>
      outlineSettings {
       bUseBrushTransparency=false
       roundingType=HalfHeightRadius
       width=0
       color {colorUseRule=UseColor_Specified, specifiedColor=<rgba 0 0 0 0>}
       cornerRadii=<0 0 0 0>
      }
      tintColor {colorUseRule=UseColor_Specified, specifiedColor=<rgba 0 0.0802 0.3515 1>}
      uVRegion {bIsValid=false, max=<0 0>, min=<0 0>}
     }
    }
   }
   slot {class=HorizontalBoxSlot}
  }
  Border {
   widgetClassPath=&Border
   parent=HorizontalBox_2
   widget=Border
   properties {brushColor=<rgba 0.0395 0.0395 0.0395 1>, contentColorAndOpacity=<rgba 0.0395 0.0395 0.0395 1>, padding=<ltrb 1.5 0 0 0>}
   slot {class=HorizontalBoxSlot, properties {padding=<ltrb 3 0 3 0>}}
  }
  UEFN_TextBlock_C {
   widgetClassPath=&UEFN_TextBlock
   parent=HorizontalBox_2
   widget=UEFN_TextBlock_C
   properties {
    text=Ground:
    font {
     bForceMonospaced=false
     bMaterialIsStencil=false
     fontMaterial=null
     fontObject=&Roboto
     letterSpacing=0
     monospacedWidth=1
     size=6.75
     skewAmount=0
     typefaceFontName=Regular
     outlineSettings {
      bApplyOutlineToDropShadows=false
      bMiteredCorners=false
      bSeparateFillAlpha=false
      outlineMaterial=null
      outlineSize=0
      outlineColor=<rgba 0 0 0 1>
     }
    }
    renderTransform {angle=0, scale=<1 1>, shear=<0 0>, translation=<0 0.5>}
   }
   slot {class=HorizontalBoxSlot, properties {horizontalAlignment=HAlign_Right, verticalAlignment=VAlign_Top, padding=<ltrb 0 3 0 0>}}
  }
  HBox_Stage {
   widgetClassPath=&HorizontalBox
   parent=HorizontalBox_2
   widget=HBox_Stage
   slot {class=HorizontalBoxSlot, properties {horizontalAlignment=HAlign_Left, verticalAlignment=VAlign_Top}}
  }
  VerticalBox_5 {
   widgetClassPath=&VerticalBox
   parent=HBox_Stage
   widget=VerticalBox_5
   slot {class=HorizontalBoxSlot, properties {horizontalAlignment=HAlign_Left, verticalAlignment=VAlign_Center}}
  }
  HorizontalBox_5 {widgetClassPath=&HorizontalBox, parent=VerticalBox_5, widget=HorizontalBox_5, slot {class=VerticalBoxSlot}}
  TextBlock_4 {
   widgetClassPath=&TextBlock
   parent=HorizontalBox_5
   widget=TextBlock_4
   properties {
    text=Show Stage
    font {
     bForceMonospaced=false
     bMaterialIsStencil=false
     fontMaterial=null
     fontObject=&Roboto
     letterSpacing=0
     monospacedWidth=1
     size=6
     skewAmount=0
     typefaceFontName=Regular
     outlineSettings {
      bApplyOutlineToDropShadows=false
      bMiteredCorners=false
      bSeparateFillAlpha=false
      outlineMaterial=null
      outlineSize=0
      outlineColor=<rgba 0 0 0 1>
     }
    }
   }
   slot {class=HorizontalBoxSlot, properties {horizontalAlignment=HAlign_Center, verticalAlignment=VAlign_Center, size {sizeRule=Fill, value=1}}}
  }
  ShowStage_CHK {
   widgetClassPath=&EditorUtilityCheckBox
   bIsVariable=true
   parent=HorizontalBox_5
   widget=ShowStage_CHK
   properties {
    horizontalAlignment=HAlign_Left
    renderTransform {angle=0, scale=<1 0.5>, shear=<0 0>, translation=<0 0>}
    widgetStyle {
     checkBoxType=CheckBox
     backgroundHoveredImage {
      bIsDynamicallyLoaded=false
      drawAs=Box
      imageType=NoImage
      mirroring=NoMirror
      resourceName=None
      resourceObject=null
      tiling=NoTile
      imageSize=<18 18>
      margin=<ltrb 0 0 0 0>
      outlineSettings {
       bUseBrushTransparency=false
       roundingType=FixedRadius
       width=1
       color {colorUseRule=UseColor_ColorTable, specifiedColor=<rgba 0 0 0 0>}
       cornerRadii=<3 3 3 3>
      }
      tintColor {colorUseRule=UseColor_Specified, specifiedColor=<rgba 0.0048 0.0048 0.0048 0>}
      uVRegion {bIsValid=false, max=<0 0>, min=<0 0>}
     }
     backgroundImage {
      bIsDynamicallyLoaded=false
      drawAs=RoundedBox
      imageType=NoImage
      mirroring=NoMirror
      resourceName=None
      resourceObject=null
      tiling=NoTile
      imageSize=<18 18>
      margin=<ltrb 0 0 0 0>
      outlineSettings {
       bUseBrushTransparency=false
       roundingType=FixedRadius
       width=0
       color {colorUseRule=UseColor_ColorTable, specifiedColor=<rgba 0 0 0 0>}
       cornerRadii=<3 3 3 3>
      }
      tintColor {colorUseRule=UseColor_Specified, specifiedColor=<rgba 0.0048 0.0048 0.0048 0>}
      uVRegion {bIsValid=false, max=<0 0>, min=<0 0>}
     }
     backgroundPressedImage {
      bIsDynamicallyLoaded=false
      drawAs=RoundedBox
      imageType=NoImage
      mirroring=NoMirror
      resourceName=None
      resourceObject=null
      tiling=NoTile
      imageSize=<18 18>
      margin=<ltrb 0 0 0 0>
      outlineSettings {
       bUseBrushTransparency=false
       roundingType=FixedRadius
       width=0
       color {colorUseRule=UseColor_ColorTable, specifiedColor=<rgba 0 0 0 0>}
       cornerRadii=<3 3 3 3>
      }
      tintColor {colorUseRule=UseColor_Specified, specifiedColor=<rgba 0.0048 0.0048 0.0048 0>}
      uVRegion {bIsValid=false, max=<0 0>, min=<0 0>}
     }
     borderBackgroundColor {colorUseRule=UseColor_Specified, specifiedColor=<rgba 1 1 1 0>}
     checkedForeground {colorUseRule=UseColor_Specified, specifiedColor=<rgba 1 1 1 1>}
     checkedHoveredForeground {colorUseRule=UseColor_Specified, specifiedColor=<rgba 1 1 1 1>}
     checkedHoveredImage {
      bIsDynamicallyLoaded=false
      drawAs=Image
      imageType=FullColor
      mirroring=NoMirror
      resourceName=None
      resourceObject=&AnimToolbox_UI_Icon_toggle_on
      tiling=NoTile
      imageSize=<8 4>
      margin=<ltrb 0 0 0 0>
      outlineSettings {
       bUseBrushTransparency=false
       roundingType=HalfHeightRadius
       width=0
       color {colorUseRule=UseColor_Specified, specifiedColor=<rgba 0 0 0 0>}
       cornerRadii=<0 0 0 0>
      }
      tintColor {colorUseRule=UseColor_Specified, specifiedColor=<rgba 1 1 1 1>}
      uVRegion {bIsValid=false, max=<0 0>, min=<0 0>}
     }
     checkedImage {
      bIsDynamicallyLoaded=false
      drawAs=Image
      imageType=FullColor
      mirroring=NoMirror
      resourceName=None
      resourceObject=&AnimToolbox_UI_Icon_toggle_on
      tiling=NoTile
      imageSize=<8 4>
      margin=<ltrb 0 0 0 0>
      outlineSettings {
       bUseBrushTransparency=false
       roundingType=HalfHeightRadius
       width=0
       color {colorUseRule=UseColor_Specified, specifiedColor=<rgba 0 0 0 0>}
       cornerRadii=<0 0 0 0>
      }
      tintColor {colorUseRule=UseColor_Specified, specifiedColor=<rgba 0.6658 0.6658 0.6658 1>}
      uVRegion {bIsValid=false, max=<0 0>, min=<0 0>}
     }
     checkedPressedForeground {colorUseRule=UseColor_Specified, specifiedColor=<rgba 1 1 1 1>}
     checkedPressedImage {
      bIsDynamicallyLoaded=false
      drawAs=Image
      imageType=FullColor
      mirroring=NoMirror
      resourceName=None
      resourceObject=&AnimToolbox_UI_Icon_toggle_off
      tiling=NoTile
      imageSize=<8 4>
      margin=<ltrb 0 0 0 0>
      outlineSettings {
       bUseBrushTransparency=false
       roundingType=HalfHeightRadius
       width=0
       color {colorUseRule=UseColor_Specified, specifiedColor=<rgba 0 0 0 0>}
       cornerRadii=<0 0 0 0>
      }
      tintColor {colorUseRule=UseColor_Specified, specifiedColor=<rgba 1 1 1 1>}
      uVRegion {bIsValid=false, max=<0 0>, min=<0 0>}
     }
     checkedSlateSound {resourceObject=null}
     foregroundColor {colorUseRule=UseColor_Specified, specifiedColor=<rgba 1 1 1 1>}
     hoveredForeground {colorUseRule=UseColor_Specified, specifiedColor=<rgba 1 1 1 1>}
     hoveredSlateSound {resourceObject=null}
     padding=<ltrb 0 0 0 0>
     pressedForeground {colorUseRule=UseColor_Specified, specifiedColor=<rgba 1 1 1 1>}
     uncheckedHoveredImage {
      bIsDynamicallyLoaded=false
      drawAs=Image
      imageType=FullColor
      mirroring=NoMirror
      resourceName=None
      resourceObject=&AnimToolbox_UI_Icon_toggle_off
      tiling=NoTile
      imageSize=<8 4>
      margin=<ltrb 0 0 0 0>
      outlineSettings {
       bUseBrushTransparency=false
       roundingType=HalfHeightRadius
       width=0
       color {colorUseRule=UseColor_Specified, specifiedColor=<rgba 0 0 0 0>}
       cornerRadii=<0 0 0 0>
      }
      tintColor {colorUseRule=UseColor_Specified, specifiedColor=<rgba 1 1 1 1>}
      uVRegion {bIsValid=false, max=<0 0>, min=<0 0>}
     }
     uncheckedImage {
      bIsDynamicallyLoaded=false
      drawAs=Image
      imageType=FullColor
      mirroring=NoMirror
      resourceName=None
      resourceObject=&AnimToolbox_UI_Icon_toggle_off
      tiling=NoTile
      imageSize=<8 4>
      margin=<ltrb 0 0 0 0>
      outlineSettings {
       bUseBrushTransparency=false
       roundingType=HalfHeightRadius
       width=0
       color {colorUseRule=UseColor_Specified, specifiedColor=<rgba 0 0 0 0>}
       cornerRadii=<0 0 0 0>
      }
      tintColor {colorUseRule=UseColor_Specified, specifiedColor=<rgba 0.666 0.666 0.666 1>}
      uVRegion {bIsValid=false, max=<0 0>, min=<0 0>}
     }
     uncheckedPressedImage {
      bIsDynamicallyLoaded=false
      drawAs=Image
      imageType=FullColor
      mirroring=NoMirror
      resourceName=None
      resourceObject=&AnimToolbox_UI_Icon_toggle_off
      tiling=NoTile
      imageSize=<8 4>
      margin=<ltrb 0 0 0 0>
      outlineSettings {
       bUseBrushTransparency=false
       roundingType=HalfHeightRadius
       width=0
       color {colorUseRule=UseColor_Specified, specifiedColor=<rgba 0 0 0 0>}
       cornerRadii=<0 0 0 0>
      }
      tintColor {colorUseRule=UseColor_Specified, specifiedColor=<rgba 1 1 1 1>}
      uVRegion {bIsValid=false, max=<0 0>, min=<0 0>}
     }
     uncheckedSlateSound {resourceObject=null}
     undeterminedForeground {colorUseRule=UseColor_Specified, specifiedColor=<rgba 1 1 1 1>}
     undeterminedHoveredImage {
      bIsDynamicallyLoaded=false
      drawAs=Image
      imageType=Vector
      mirroring=NoMirror
      resourceName=../../../Engine/Content/Slate/Starship/CoreWidgets/CheckBox/indeterminate.svg
      resourceObject=null
      tiling=NoTile
      imageSize=<16 16>
      margin=<ltrb 0 0 0 0>
      outlineSettings {
       bUseBrushTransparency=false
       roundingType=HalfHeightRadius
       width=0
       color {colorUseRule=UseColor_Specified, specifiedColor=<rgba 0 0 0 0>}
       cornerRadii=<0 0 0 0>
      }
      tintColor {colorUseRule=UseColor_Specified, specifiedColor=<rgba 0.0044 0.2384 1 1>}
      uVRegion {bIsValid=false, max=<0 0>, min=<0 0>}
     }
     undeterminedImage {
      bIsDynamicallyLoaded=false
      drawAs=Image
      imageType=Vector
      mirroring=NoMirror
      resourceName=../../../Engine/Content/Slate/Starship/CoreWidgets/CheckBox/indeterminate.svg
      resourceObject=null
      tiling=NoTile
      imageSize=<16 16>
      margin=<ltrb 0 0 0 0>
      outlineSettings {
       bUseBrushTransparency=false
       roundingType=HalfHeightRadius
       width=0
       color {colorUseRule=UseColor_Specified, specifiedColor=<rgba 0 0 0 0>}
       cornerRadii=<0 0 0 0>
      }
      tintColor {colorUseRule=UseColor_Specified, specifiedColor=<rgba 0 0.162 0.7454 1>}
      uVRegion {bIsValid=false, max=<0 0>, min=<0 0>}
     }
     undeterminedPressedImage {
      bIsDynamicallyLoaded=false
      drawAs=Image
      imageType=Vector
      mirroring=NoMirror
      resourceName=../../../Engine/Content/Slate/Starship/CoreWidgets/CheckBox/indeterminate.svg
      resourceObject=null
      tiling=NoTile
      imageSize=<16 16>
      margin=<ltrb 0 0 0 0>
      outlineSettings {
       bUseBrushTransparency=false
       roundingType=HalfHeightRadius
       width=0
       color {colorUseRule=UseColor_Specified, specifiedColor=<rgba 0 0 0 0>}
       cornerRadii=<0 0 0 0>
      }
      tintColor {colorUseRule=UseColor_Specified, specifiedColor=<rgba 0 0.0802 0.3515 1>}
      uVRegion {bIsValid=false, max=<0 0>, min=<0 0>}
     }
    }
   }
   slot {class=HorizontalBoxSlot, properties {horizontalAlignment=HAlign_Center, verticalAlignment=VAlign_Center}}
  }
  HorizontalBox_4 {widgetClassPath=&HorizontalBox, parent=VerticalBox_5, widget=HorizontalBox_4, slot {class=VerticalBoxSlot}}
  TextBlock_2 {
   widgetClassPath=&TextBlock
   parent=HorizontalBox_4
   widget=TextBlock_2
   properties {
    text=Show Grid
    font {
     bForceMonospaced=false
     bMaterialIsStencil=false
     fontMaterial=null
     fontObject=&Roboto
     letterSpacing=0
     monospacedWidth=1
     size=6
     skewAmount=0
     typefaceFontName=Regular
     outlineSettings {
      bApplyOutlineToDropShadows=false
      bMiteredCorners=false
      bSeparateFillAlpha=false
      outlineMaterial=null
      outlineSize=0
      outlineColor=<rgba 0 0 0 1>
     }
    }
   }
   slot {class=HorizontalBoxSlot, properties {horizontalAlignment=HAlign_Center, verticalAlignment=VAlign_Center, size {sizeRule=Fill, value=1}}}
  }
  ShowGrid_CHK {
   widgetClassPath=&EditorUtilityCheckBox
   bIsVariable=true
   parent=HorizontalBox_4
   widget=ShowGrid_CHK
   properties {
    horizontalAlignment=HAlign_Left
    renderTransform {angle=0, scale=<1 0.5>, shear=<0 0>, translation=<0 0>}
    widgetStyle {
     checkBoxType=CheckBox
     backgroundHoveredImage {
      bIsDynamicallyLoaded=false
      drawAs=Box
      imageType=NoImage
      mirroring=NoMirror
      resourceName=None
      resourceObject=null
      tiling=NoTile
      imageSize=<18 18>
      margin=<ltrb 0 0 0 0>
      outlineSettings {
       bUseBrushTransparency=false
       roundingType=FixedRadius
       width=1
       color {colorUseRule=UseColor_ColorTable, specifiedColor=<rgba 0 0 0 0>}
       cornerRadii=<3 3 3 3>
      }
      tintColor {colorUseRule=UseColor_Specified, specifiedColor=<rgba 0.0048 0.0048 0.0048 0>}
      uVRegion {bIsValid=false, max=<0 0>, min=<0 0>}
     }
     backgroundImage {
      bIsDynamicallyLoaded=false
      drawAs=RoundedBox
      imageType=NoImage
      mirroring=NoMirror
      resourceName=None
      resourceObject=null
      tiling=NoTile
      imageSize=<18 18>
      margin=<ltrb 0 0 0 0>
      outlineSettings {
       bUseBrushTransparency=false
       roundingType=FixedRadius
       width=0
       color {colorUseRule=UseColor_ColorTable, specifiedColor=<rgba 0 0 0 0>}
       cornerRadii=<3 3 3 3>
      }
      tintColor {colorUseRule=UseColor_Specified, specifiedColor=<rgba 0.0048 0.0048 0.0048 0>}
      uVRegion {bIsValid=false, max=<0 0>, min=<0 0>}
     }
     backgroundPressedImage {
      bIsDynamicallyLoaded=false
      drawAs=RoundedBox
      imageType=NoImage
      mirroring=NoMirror
      resourceName=None
      resourceObject=null
      tiling=NoTile
      imageSize=<18 18>
      margin=<ltrb 0 0 0 0>
      outlineSettings {
       bUseBrushTransparency=false
       roundingType=FixedRadius
       width=0
       color {colorUseRule=UseColor_ColorTable, specifiedColor=<rgba 0 0 0 0>}
       cornerRadii=<3 3 3 3>
      }
      tintColor {colorUseRule=UseColor_Specified, specifiedColor=<rgba 0.0048 0.0048 0.0048 0>}
      uVRegion {bIsValid=false, max=<0 0>, min=<0 0>}
     }
     borderBackgroundColor {colorUseRule=UseColor_Specified, specifiedColor=<rgba 1 1 1 0>}
     checkedForeground {colorUseRule=UseColor_Specified, specifiedColor=<rgba 1 1 1 1>}
     checkedHoveredForeground {colorUseRule=UseColor_Specified, specifiedColor=<rgba 1 1 1 1>}
     checkedHoveredImage {
      bIsDynamicallyLoaded=false
      drawAs=Image
      imageType=FullColor
      mirroring=NoMirror
      resourceName=None
      resourceObject=&AnimToolbox_UI_Icon_toggle_on
      tiling=NoTile
      imageSize=<8 4>
      margin=<ltrb 0 0 0 0>
      outlineSettings {
       bUseBrushTransparency=false
       roundingType=HalfHeightRadius
       width=0
       color {colorUseRule=UseColor_Specified, specifiedColor=<rgba 0 0 0 0>}
       cornerRadii=<0 0 0 0>
      }
      tintColor {colorUseRule=UseColor_Specified, specifiedColor=<rgba 1 1 1 1>}
      uVRegion {bIsValid=false, max=<0 0>, min=<0 0>}
     }
     checkedImage {
      bIsDynamicallyLoaded=false
      drawAs=Image
      imageType=FullColor
      mirroring=NoMirror
      resourceName=None
      resourceObject=&AnimToolbox_UI_Icon_toggle_on
      tiling=NoTile
      imageSize=<8 4>
      margin=<ltrb 0 0 0 0>
      outlineSettings {
       bUseBrushTransparency=false
       roundingType=HalfHeightRadius
       width=0
       color {colorUseRule=UseColor_Specified, specifiedColor=<rgba 0 0 0 0>}
       cornerRadii=<0 0 0 0>
      }
      tintColor {colorUseRule=UseColor_Specified, specifiedColor=<rgba 0.6658 0.6658 0.6658 1>}
      uVRegion {bIsValid=false, max=<0 0>, min=<0 0>}
     }
     checkedPressedForeground {colorUseRule=UseColor_Specified, specifiedColor=<rgba 1 1 1 1>}
     checkedPressedImage {
      bIsDynamicallyLoaded=false
      drawAs=Image
      imageType=FullColor
      mirroring=NoMirror
      resourceName=None
      resourceObject=&AnimToolbox_UI_Icon_toggle_off
      tiling=NoTile
      imageSize=<8 4>
      margin=<ltrb 0 0 0 0>
      outlineSettings {
       bUseBrushTransparency=false
       roundingType=HalfHeightRadius
       width=0
       color {colorUseRule=UseColor_Specified, specifiedColor=<rgba 0 0 0 0>}
       cornerRadii=<0 0 0 0>
      }
      tintColor {colorUseRule=UseColor_Specified, specifiedColor=<rgba 1 1 1 1>}
      uVRegion {bIsValid=false, max=<0 0>, min=<0 0>}
     }
     checkedSlateSound {resourceObject=null}
     foregroundColor {colorUseRule=UseColor_Specified, specifiedColor=<rgba 1 1 1 1>}
     hoveredForeground {colorUseRule=UseColor_Specified, specifiedColor=<rgba 1 1 1 1>}
     hoveredSlateSound {resourceObject=null}
     padding=<ltrb 0 0 0 0>
     pressedForeground {colorUseRule=UseColor_Specified, specifiedColor=<rgba 1 1 1 1>}
     uncheckedHoveredImage {
      bIsDynamicallyLoaded=false
      drawAs=Image
      imageType=FullColor
      mirroring=NoMirror
      resourceName=None
      resourceObject=&AnimToolbox_UI_Icon_toggle_off
      tiling=NoTile
      imageSize=<8 4>
      margin=<ltrb 0 0 0 0>
      outlineSettings {
       bUseBrushTransparency=false
       roundingType=HalfHeightRadius
       width=0
       color {colorUseRule=UseColor_Specified, specifiedColor=<rgba 0 0 0 0>}
       cornerRadii=<0 0 0 0>
      }
      tintColor {colorUseRule=UseColor_Specified, specifiedColor=<rgba 1 1 1 1>}
      uVRegion {bIsValid=false, max=<0 0>, min=<0 0>}
     }
     uncheckedImage {
      bIsDynamicallyLoaded=false
      drawAs=Image
      imageType=FullColor
      mirroring=NoMirror
      resourceName=None
      resourceObject=&AnimToolbox_UI_Icon_toggle_off
      tiling=NoTile
      imageSize=<8 4>
      margin=<ltrb 0 0 0 0>
      outlineSettings {
       bUseBrushTransparency=false
       roundingType=HalfHeightRadius
       width=0
       color {colorUseRule=UseColor_Specified, specifiedColor=<rgba 0 0 0 0>}
       cornerRadii=<0 0 0 0>
      }
      tintColor {colorUseRule=UseColor_Specified, specifiedColor=<rgba 0.666 0.666 0.666 1>}
      uVRegion {bIsValid=false, max=<0 0>, min=<0 0>}
     }
     uncheckedPressedImage {
      bIsDynamicallyLoaded=false
      drawAs=Image
      imageType=FullColor
      mirroring=NoMirror
      resourceName=None
      resourceObject=&AnimToolbox_UI_Icon_toggle_off
      tiling=NoTile
      imageSize=<8 4>
      margin=<ltrb 0 0 0 0>
      outlineSettings {
       bUseBrushTransparency=false
       roundingType=HalfHeightRadius
       width=0
       color {colorUseRule=UseColor_Specified, specifiedColor=<rgba 0 0 0 0>}
       cornerRadii=<0 0 0 0>
      }
      tintColor {colorUseRule=UseColor_Specified, specifiedColor=<rgba 1 1 1 1>}
      uVRegion {bIsValid=false, max=<0 0>, min=<0 0>}
     }
     uncheckedSlateSound {resourceObject=null}
     undeterminedForeground {colorUseRule=UseColor_Specified, specifiedColor=<rgba 1 1 1 1>}
     undeterminedHoveredImage {
      bIsDynamicallyLoaded=false
      drawAs=Image
      imageType=Vector
      mirroring=NoMirror
      resourceName=../../../Engine/Content/Slate/Starship/CoreWidgets/CheckBox/indeterminate.svg
      resourceObject=null
      tiling=NoTile
      imageSize=<16 16>
      margin=<ltrb 0 0 0 0>
      outlineSettings {
       bUseBrushTransparency=false
       roundingType=HalfHeightRadius
       width=0
       color {colorUseRule=UseColor_Specified, specifiedColor=<rgba 0 0 0 0>}
       cornerRadii=<0 0 0 0>
      }
      tintColor {colorUseRule=UseColor_Specified, specifiedColor=<rgba 0.0044 0.2384 1 1>}
      uVRegion {bIsValid=false, max=<0 0>, min=<0 0>}
     }
     undeterminedImage {
      bIsDynamicallyLoaded=false
      drawAs=Image
      imageType=Vector
      mirroring=NoMirror
      resourceName=../../../Engine/Content/Slate/Starship/CoreWidgets/CheckBox/indeterminate.svg
      resourceObject=null
      tiling=NoTile
      imageSize=<16 16>
      margin=<ltrb 0 0 0 0>
      outlineSettings {
       bUseBrushTransparency=false
       roundingType=HalfHeightRadius
       width=0
       color {colorUseRule=UseColor_Specified, specifiedColor=<rgba 0 0 0 0>}
       cornerRadii=<0 0 0 0>
      }
      tintColor {colorUseRule=UseColor_Specified, specifiedColor=<rgba 0 0.162 0.7454 1>}
      uVRegion {bIsValid=false, max=<0 0>, min=<0 0>}
     }
     undeterminedPressedImage {
      bIsDynamicallyLoaded=false
      drawAs=Image
      imageType=Vector
      mirroring=NoMirror
      resourceName=../../../Engine/Content/Slate/Starship/CoreWidgets/CheckBox/indeterminate.svg
      resourceObject=null
      tiling=NoTile
      imageSize=<16 16>
      margin=<ltrb 0 0 0 0>
      outlineSettings {
       bUseBrushTransparency=false
       roundingType=HalfHeightRadius
       width=0
       color {colorUseRule=UseColor_Specified, specifiedColor=<rgba 0 0 0 0>}
       cornerRadii=<0 0 0 0>
      }
      tintColor {colorUseRule=UseColor_Specified, specifiedColor=<rgba 0 0.0802 0.3515 1>}
      uVRegion {bIsValid=false, max=<0 0>, min=<0 0>}
     }
    }
   }
   slot {class=HorizontalBoxSlot, properties {horizontalAlignment=HAlign_Center, verticalAlignment=VAlign_Center}}
  }
 ]
}
