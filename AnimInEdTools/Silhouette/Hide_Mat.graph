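# Hide_Mat is a translucent Default Lit material in the Silhouette tools, usable on skeletal meshes and mesh deformers, containing a procedural grid mask.
# A Custom HLSL node draws minor and thicker major lines from world-space XY, anchored to the origin.
# An exponent-shaped pixel-depth fade attenuates the lines, then GridOpacity scales the mask.
# No material output connections appear for that mask or the separate grey vector parameter.
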Material /AnimInEdTools/Silhouette/Hide_Mat
graph {
 properties {
  bUsedWithMeshDeformer=true
  bUsedWithSkeletalMesh=true
  blendMode=BLEND_Translucent
  pixelDepthOffsetMode=PDOM_Legacy
  usedShadingModels=Default Lit
 }
 nodes [
  MaterialExpressionWorldPosition_0 = WorldPosition()
  MaterialExpressionScalarParameter_1 = ScalarParameter() {defaultValue=25, group=Grid, parameterName=GridSize, sliderMax=1000, sliderMin=0.1}
  MaterialExpressionScalarParameter_2 = ScalarParameter() {defaultValue=0.005, group=Grid, parameterName=GridThickness, sliderMax=0.5, sliderMin=0.01}
  MaterialExpressionScalarParameter_7 = ScalarParameter() {defaultValue=4, group=Grid, parameterName=MajorGridScale, sliderMax=100, sliderMin=2}
  MaterialExpressionScalarParameter_5 = ScalarParameter() {defaultValue=5, group=Grid, parameterName=MajorLineThicknessMultiplier, sliderMax=10, sliderMin=1}
  MaterialExpressionPixelDepth_0 = PixelDepth()
  MaterialExpressionScalarParameter_3 = ScalarParameter() {defaultValue=1000, group=Grid, parameterName=DepthFadeStart, sliderMax=10000}
  MaterialExpressionScalarParameter_4 = ScalarParameter() {defaultValue=500, group=Grid, parameterName=DepthFadeRange, sliderMax=10000, sliderMin=1}
  MaterialExpressionScalarParameter_6 = ScalarParameter() {defaultValue=0.6, group=Grid, parameterName=DepthFalloffExp, sliderMax=8, sliderMin=0.5}
  n1 = Custom!(
   WorldPos=@MaterialExpressionWorldPosition_0.XYZ
   GridSize=@MaterialExpressionScalarParameter_1
   Thickness=@MaterialExpressionScalarParameter_2
   MajorScale=@MaterialExpressionScalarParameter_7
   MajorThickMult=@MaterialExpressionScalarParameter_5
   Depth=@MaterialExpressionPixelDepth_0.R
   FadeStart=@MaterialExpressionScalarParameter_3
   FadeRange=@MaterialExpressionScalarParameter_4
   FalloffExp=@MaterialExpressionScalarParameter_6
  ) {
   code="// World-space UVs anchored to origin\nfloat2 uv = WorldPos.xy / max(GridSize, 0.001);\n\n// Minor grid lines\nfloat2 f = frac(uv);\nfloat2 df = abs(f - round(f));\nfloat gx = 1.0 - step(Thickness, df.x);\nfloat gy = 1.0 - step(Thickness, df.y);\nfloat minorGrid = saturate(gx + gy);\n\n// Major grid lines — every MajorScale minor tiles\nfloat2 uvMajor = uv / max(MajorScale, 1.0);\nfloat2 fM = frac(uvMajor);\nfloat2 dfM = abs(fM - round(fM));\nfloat majorThickness = Thickness * MajorThickMult;\nfloat gxM = 1.0 - step(majorThickness, dfM.x * max(MajorScale, 1.0));\nfloat gyM = 1.0 - step(majorThickness, dfM.y * max(MajorScale, 1.0));\nfloat majorGrid = saturate(gxM + gyM);\n\nfloat grid = saturate(minorGrid + majorGrid);\n\n// Exponential depth falloff\nfloat t = saturate((Depth - FadeStart) / max(FadeRange, 0.001));\nfloat fade = 1.0 - pow(t, max(FalloffExp, 0.01));\n\nreturn grid * fade;"
   desc=Grid mask with depth falloff
   outputType=CMOT_Float1
   inputs [
    {
     inputName=WorldPos
     input {expression=&self:MaterialExpressionWorldPosition_0, mask=1, maskB=1, maskG=1, maskR=1}
    }
    {inputName=GridSize, input {expression=&self:MaterialExpressionScalarParameter_1}}
    {inputName=Thickness, input {expression=&self:MaterialExpressionScalarParameter_2}}
    {inputName=MajorScale, input {expression=&self:MaterialExpressionScalarParameter_7}}
    {inputName=MajorThickMult, input {expression=&self:MaterialExpressionScalarParameter_5}}
    {inputName=Depth, input {expression=&self:MaterialExpressionPixelDepth_0, mask=1, maskR=1}}
    {inputName=FadeStart, input {expression=&self:MaterialExpressionScalarParameter_3}}
    {inputName=FadeRange, input {expression=&self:MaterialExpressionScalarParameter_4}}
    {inputName=FalloffExp, input {expression=&self:MaterialExpressionScalarParameter_6}}
   ]
  }
  Multiply(A=@n1, B=ScalarParameter() {group=Grid, parameterName=GridOpacity, sliderMax=1})
  VectorParameter() {defaultValue=<rgba 0.3 0.3 0.3 1>}
 ]
}
