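# M_IsolateEnviroSphere is a two-sided unlit material in the AnimInEdTools Silhouette folder, likely for an environment sphere that isolates animation silhouettes.
# It multiplies a dark grey Base Color vector parameter by an Emissive Intensity scalar to produce a flat, adjustable color.
# Ray-traced shadow casting is disabled.
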Material /AnimInEdTools/Silhouette/M_IsolateEnviroSphere
graph {
 properties {bCastRayTracedShadows=false, shadingModel=MSM_Unlit, twoSided=true, usedShadingModels=Unlit}
 nodes [Multiply(
   A=VectorParameter() {parameterName=Base Color, defaultValue=<rgba 0.026 0.026 0.026 0>}.RGB
   B=ScalarParameter() {defaultValue=1, parameterName=Emissive Intensity}
  )]
}
