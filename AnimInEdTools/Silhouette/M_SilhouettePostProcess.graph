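# Post-process material for the AnimInEdTools silhouette feature, apparently meant to tint the screen with a silhouette color.
# A LinearInterpolate takes its B input from the SilhouetteColor parameter in the MPC_Silhouette collection and its Alpha from a Saturate node.
# The graph shows no source for the A input or the Saturate, no scene texture sampling and no link to a material output.
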Material /AnimInEdTools/Silhouette/M_SilhouettePostProcess
graph {
 properties {bAutomaticallySetUsageInEditor=false, materialDomain=MD_PostProcess, usedShadingModels=Default Lit}
 nodes [LinearInterpolate(B=CollectionParameter() {collection=&/MPC_Silhouette, parameterName=SilhouetteColor}, Alpha=Saturate())]
}
